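; Editor Blueprint actor that bakes tiling caustics textures from Gerstner waves, as single frames or SubUV sheets animated over time or water depth.
; A photon grid with scatter materials feeds an overscanned scene capture, and editor tick copies successive frames into the sheet.
; Downsampled mip chains supply average and maximum luminance for normalization before textures are saved as HDR or LDR.
; Results preview on water meshes or through a directional-light function.
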(event EventBeginPlay
  (Actor|Tick|SetActorTickEnabled false))

(event Tick|EventEditorTick
  (params (DeltaSeconds :type Float%20%28single-precision%29))
  (bind (_looped_time _blend_phase) (Utilities|LoopingTime))
  (bind _caustics_scatter_mid_singleframe (Variables|Debug|MIDs|GetCaustics%20Scatter%20MID%20SingleFrame))
  (bind (_x _y) (Utilities|Struct|BreakIntPoint (Variables|OutputandAnimation|GetAnimation%20Frames)))
  (bind _current_frame (Variables|Default|GetCurrent%20Frame))
  (if (Variables|Debug|GetTick%20For%20Capture%20Only)
    (CallFunction|UpdateCapture)
    (CallFunction|NormalizeCausticsLuminance true)
    (Default|NormalizeByMaxLuminance true)
    (Variables|Debug|SetTick%20For%20Capture%20Only false)
    (Default|SetEditorTickEnabled false)
    (CallFunction|SetLightFunctionCaustics false)
    (OutputandAnimation|StopRendering)
    (else
      ; Prepare Tiling & Looping Normalmap
      (Rendering|Material|SetTextureParameterValue (Variables|Debug|MIDs|GetWaterSurface%202D%20MID) "Caustics" (Variables|RenderTargets|GetCaustics%20RT))
      (Rendering|Material|SetScalarParameterValue "/Water/Materials/MPC/MPC_Water.MPC_Water" "FreezeTime" (Math|Conversions|ToFloat(Boolean) (Utilities|Select :Index (Variables|OutputandAnimation|GetOutput%20Mode) :NewEnumerator0 false :NewEnumerator1 true :NewEnumerator2 true)) :!class /Script/Engine.KismetMaterialLibrary)
      (Rendering|Material|SetScalarParameterValue "/Water/Materials/MPC/MPC_Water.MPC_Water" "ManualTime" (+ (Variables|Default|GetTime%20Accumulator) (Variables|Default|GetTime%20Offset)) :!class /Script/Engine.KismetMaterialLibrary)
      (if (Utilities|Enum|Equal(Enum) (Variables|OutputandAnimation|GetOutput%20Mode) "NewEnumerator1")
        (Rendering|Material|SetScalarParameterValue (Variables|Debug|MIDs|GetRender%20Normal%20MID) "FrameAlpha" 1.0 :!class /Script/Engine.MaterialInstanceDynamic)
        (Rendering|ClearRenderTarget2D (Variables|Debug|GetWaterNormal))
        (Rendering|DrawMaterialtoRenderTarget (Variables|Debug|GetWaterNormal) (Variables|Debug|MIDs|GetRender%20Normal%20MID))
        (Rendering|Material|SetScalarParameterValue "/Water/Materials/MPC/MPC_Water.MPC_Water" "ManualTime" (+ _looped_time (Variables|Default|GetTime%20Offset)) :!class /Script/Engine.KismetMaterialLibrary)
        (Rendering|Material|SetScalarParameterValue (Variables|Debug|MIDs|GetRender%20Normal%20MID) "FrameAlpha" _blend_phase :!class /Script/Engine.MaterialInstanceDynamic)
        (Rendering|DrawMaterialtoRenderTarget (Variables|Debug|GetWaterNormal) (Variables|Debug|MIDs|GetRender%20Normal%20MID)))
      ; Animate Water Depth
      (if (Utilities|Enum|Equal(Enum) (Variables|OutputandAnimation|GetOutput%20Mode) "NewEnumerator2")
        (Rendering|Material|SetScalarParameterValue _caustics_scatter_mid_singleframe "WaterDepth" (Variables|Default|GetCurrentWaterDepth) :!class /Script/Engine.MaterialInstanceDynamic)
        (Rendering|Material|SetScalarParameterValue _caustics_scatter_mid_singleframe "Freeze Time" 1.0 :!class /Script/Engine.MaterialInstanceDynamic)
        (else
          _))
      (CallFunction|UpdateCapture)
      (if (< (Variables|Default|GetCurrent%20Frame) (Variables|Debug|GetAnimationFrames))
        ; Draw SubUV Frame
        (bind (_canvas _size _context) (Rendering|BeginDrawCanvastoRenderTarget (Variables|RenderTargets|GetSubUV%20RT)))
        (Utilities|IsValid _canvas
          (:Is%20Valid
            (Canvas|DrawMaterial :self _canvas :RenderMaterial (Variables|Debug|MIDs|GetCopy%20SubUV%20Frame%20MID) :ScreenPosition (* (/ _size (Math|Vector2D|MakeVector2D (Math|Conversions|ToFloat(Integer) _x) (Math|Conversions|ToFloat(Integer) _y))) (Math|Vector2D|MakeVector2D (Math|Conversions|ToFloat(Integer) (Math|Integer|%25%28Integer%29 (Variables|Default|GetCurrent%20Frame) (Utilities|Struct|BreakIntPoint (Variables|OutputandAnimation|GetAnimation%20Frames)))) (Math|Conversions|ToFloat(Integer) (/ (Variables|Default|GetCurrent%20Frame) (Utilities|Struct|BreakIntPoint (Variables|OutputandAnimation|GetAnimation%20Frames)))))) :ScreenSize (/ _size (Math|Vector2D|MakeVector2D (Math|Conversions|ToFloat(Integer) _x) (Math|Conversions|ToFloat(Integer) _y))) :CoordinateSize "(X=1.000000,Y=1.000000)" :Rotation 0.0 :PivotPoint "(X=0.500000,Y=0.500000)")
            (Rendering|EndDrawCanvastoRenderTarget _context)))
        ;[1 1 1] Count Frames; Check if done
        (Math|Integer|IncrementInt _current_frame
          (:%20%20
            (if (== _current_frame (Variables|Debug|GetAnimationFrames))
              ;[0.3528 0.6421 1] Render Complete, Normalize and display animation preview
              (Default|NormalizeSubUVLuminance)
              (Rendering|Material|SetMaterial (Variables|Default|GetWaterPreviewGrid) 0 (Variables|Debug|MIDs|GetWaterSurface%20SubUV%20MID))
              (Rendering|Material|SetScalarParameterValue (Variables|Debug|MIDs|GetPreview%20RenderTarget%20MID) "ShowSubUV" 1.0 :!class /Script/Engine.MaterialInstanceDynamic)
              (Rendering|Material|SetScalarParameterValue "/Water/Materials/MPC/MPC_Water.MPC_Water" "FreezeTime" :!class /Script/Engine.KismetMaterialLibrary)
              (else
                ;[0.3192 1 0.2543] Increment Animation Values During Render
                (switch Utilities|FlowControl|Switch|SwitchonCausticGeneratorOutputModes (Variables|OutputandAnimation|GetOutput%20Mode)
                  (:NewEnumerator1
                    (Variables|Default|SetTime%20Accumulator (+ (Variables|Default|GetTime%20Accumulator) (/ 1.0 (Variables|OutputandAnimation|GetFrames%20Per%20Second)))))
                  (:NewEnumerator2
                    (Variables|Default|SetCurrentWaterDepth (+ (Variables|Default|GetCurrentWaterDepth) (Variables|Default|GetDepthPerFrame)))))))))
        (else
          ;[1 1 1] When no more frames to render, normalize frame.
          (CallFunction|NormalizeCausticsLuminance true)
          (Default|NormalizeByMaxLuminance true)
          (CallFunction|SetLightFunctionCaustics false)
          (switch Utilities|FlowControl|Switch|SwitchonCausticGeneratorOutputModes (Variables|OutputandAnimation|GetOutput%20Mode)
            (:NewEnumerator1
              _)
            (:NewEnumerator2
              _))
          (Default|SetEditorTickEnabled false)
          (Rendering|Material|SetScalarParameterValue "/Water/Materials/MPC/MPC_Water.MPC_Water" "FreezeTime" :!class /Script/Engine.KismetMaterialLibrary)
          (CallFunction|SetLightFunctionCaustics true))))))

(event Custom|Initialize
  (bind (_x _y) (Utilities|Struct|BreakIntPoint (Variables|OutputandAnimation|GetSubUV%20Final%20Resolution)))
  ;[1 1 1] Initialize RTs and Materials
  ;; Allocate RTs
  (CallFunction|AllocatePreviewCaptureRT)
  (Rendering|CreateAndAssignRenderTarget (Variables|Debug|GetWaterNormal) (/ (Variables|OutputandAnimation|GetCaustic%20Mesh%20Resolution) 2) (/ (Variables|OutputandAnimation|GetCaustic%20Mesh%20Resolution) 2) "RTF_RGBA32f"
    (:Out
      (Rendering|CreateAndAssignRenderTarget (Variables|RenderTargets|GetSubUV%20RT) _x _y "RTF_RGBA16f"
        (:Out
          _))))
  ;; Assign RTs to MIDs
  (Rendering|Material|SetTextureParameterValue (Variables|Debug|MIDs|GetCopy%20SubUV%20Frame%20MID) "RT" (Variables|RenderTargets|GetCaustics%20RT))
  (Rendering|Material|SetTextureParameterValue (Variables|Debug|MIDs|GetCaustics%20Scatter%20MID%20Animation) "WaterNormal" (Variables|Debug|GetWaterNormal))
  (Rendering|Material|SetScalarParameterValue (Variables|Debug|MIDs|GetCaustics%20Scatter%20MID%20Animation) "Freeze Time" 0.0 :!class /Script/Engine.MaterialInstanceDynamic)
  (Rendering|Material|SetTextureParameterValue (Variables|Debug|MIDs|GetWaterSurface%20SubUV%20MID) "Caustics" (Variables|RenderTargets|GetSubUV%20RT))
  (Rendering|Material|SetMaterial (Variables|Default|GetCausticParticleGrid) 0 (select (Utilities|Enum|Equal(Enum) (Variables|OutputandAnimation|GetOutput%20Mode) "NewEnumerator1") (Variables|Debug|MIDs|GetCaustics%20Scatter%20MID%20Animation) (Variables|Debug|MIDs|GetCaustics%20Scatter%20MID%20SingleFrame)))
  (Rendering|Material|SetTextureParameterValue (Variables|Debug|MIDs|GetPreview%20RenderTarget%20MID) "RT" (Utilities|Select :Index (Variables|OutputandAnimation|GetOutput%20Mode) :NewEnumerator0 (Variables|RenderTargets|GetCaustics%20RT) :NewEnumerator1 (Variables|RenderTargets|GetSubUV%20RT) :NewEnumerator2 (Variables|RenderTargets|GetSubUV%20RT)))
  (Actor|Tick|SetActorTickEnabled true))

(event Custom|Allocate%20Preview%20Capture%20RT
  ; Allocate Caustics RT
  (Rendering|CreateAndAssignRenderTarget (Variables|RenderTargets|GetCaustics%20RT) (/ (Variables|OutputandAnimation|GetCaustic%20Mesh%20Resolution) (Math|Integer|Clamp(Integer) (Variables|OutputandAnimation|GetDownsample%20Output) 1 2)) (/ (Variables|OutputandAnimation|GetCaustic%20Mesh%20Resolution) (Math|Integer|Clamp(Integer) (Variables|OutputandAnimation|GetDownsample%20Output) 1 2)) "RTF_RGBA16f"
    (:Out
      (Rendering|CreateAndAssignRenderTarget (Variables|Debug|GetOverscan%20SceneCapture%20RT) (+ (Variables|OutputandAnimation|GetCaustic%20Mesh%20Resolution) (* (Variables|Caustics|GetOverscan%20Quads) 2)) (+ (Variables|OutputandAnimation|GetCaustic%20Mesh%20Resolution) (* (Variables|Caustics|GetOverscan%20Quads) 2)) "RTF_RGBA16f"
        (:Out
          _))))
  (Rendering|Material|SetTextureParameterValue (Variables|Debug|MIDs|GetPreview%20RenderTarget%20MID) "RT" (Variables|RenderTargets|GetCaustics%20RT))
  (Variables|/Script/Engine.SceneCaptureComponent2D|SetTextureTarget (Variables|Debug|GetOverscan%20SceneCapture%20RT) (Variables|Default|GetSceneCaptureComponent2D))
  (Rendering|Material|SetTextureParameterValue (Variables|Debug|MIDs|GetRemoveOverscanMID) "RT" (Variables|Debug|GetOverscan%20SceneCapture%20RT)))

(event Custom|Add%20Photon%20Meshes
  ; SpawnCausticGridMeshes
  (WaterMeshPreview|SpawnCausticParticleGrid (Variables|Default|GetCausticParticleGrid) (Variables|Caustics|GetTexture%20World%20Size) (Variables|Debug|GetTiles)))

(event Custom|Add%20Water%20Preview%20Meshes
  ; SpawnWaterPreviewMeshes
  (Utilities|FlowControl|Sequence
    (:then_0
      (WaterMeshPreview|SpawnWaterPreviewGrid (Variables|Default|GetWaterPreviewGrid) (Utilities|Select :Index (Variables|PreviewScene|GetPreview%20Scene) :NewEnumerator0 (* (Variables|Caustics|GetTexture%20World%20Size) (Variables|PreviewScene|GetPreview%20Scene%20XY%20Scale)) :NewEnumerator1 32768.0) (Math|Integer|Max(Integer) 16 1)))
    (:then_1
      (if (Utilities|Enum|Equal(Enum) (Variables|PreviewScene|GetPreview%20Scene) "NewEnumerator0")
        _
        (elif (Variables|PreviewScene|GetShow%20Water%20Scene)
          (Utilities|FlowControl|Sequence
            (:then_0
              (bind _returnvalue (AddComponent|NODE_AddStaticMeshComponent-3 :RelativeTransform (Math|Transform|MakeTransform (Math|Vector|MakeVector 0.0 0.0 (- 0.0 (Variables|PreviewScene|GetPreview%20Mesh%20Depth))) "0, 0, 0" (Math|Conversions|ToVector(Float) 1e+06))))
              (Rendering|Material|SetMaterial _returnvalue 0 (select (Utilities|IsValid (Variables|PreviewScene|GetSea%20Bed%20Material) :!class /Script/Engine.KismetSystemLibrary) (Variables|PreviewScene|GetSea%20Bed%20Material) "/Water/Caustics/Materials/Preview/CausticsBaseMesh_NoVS.CausticsBaseMesh_NoVS")))
            (:then_1
              (bind _returnvalue_2 (AddComponent|NODE_AddStaticMeshComponent-4 :RelativeTransform (Math|Transform|MakeTransform "0, 0, 0" "0, 0, 0" (Math|Conversions|ToVector(Float) (/ (- 32768.0 1024.0) 256.0)))))
              (Rendering|Material|SetMaterial _returnvalue_2 0 (Variables|Debug|MIDs|GetWaterSurface%202D%20MID))
              (Variables|Default|SetFarOceanMesh _returnvalue_2))))))))

(event Custom|TickForUpdate
  (if (not (Variables|Debug|GetTicking))
    (Variables|Debug|SetTick%20For%20Capture%20Only true)
    (Default|SetEditorTickEnabled true)))

(fn BeginRendering
  (CallFunction|Reset)
  (CallFunction|Initialize)
  (Default|SetEditorTickEnabled true)
  (Variables|Debug|SetTicking true)
  (Rendering|Material|SetScalarParameterValue "/Water/Materials/MPC/MPC_Water.MPC_Water" "FreezeTime" (Math|Conversions|ToFloat(Boolean) (Variables|PreviewScene|GetFreeze%20Time)) :!class /Script/Engine.KismetMaterialLibrary))

(fn CopyRT)

(fn GenerateMipChain
  (params (RenderTarget :type Texture%20Render%20Target%202D%20Object%20Reference%20%28by%20ref%29) (Mip%20RT%20Array :type Array%20of%20Texture%20Render%20Target%202D%20Object%20References) (Mip%20Mode :type RTMipGenerationSettings%20Enum :default "NewEnumerator0"))
  (bind _sizex (Variables|/Script/Engine.TextureRenderTarget2D|GetSizeX RenderTarget))
  (bind _sizey (Variables|/Script/Engine.TextureRenderTarget2D|GetSizeY RenderTarget))
  (bind _downsample_mid (Variables|Debug|MIDs|GetDownsample%20MID))
  (bind _downsample_mid_2 (Variables|Debug|MIDs|GetDownsample%20MID))
  (Rendering|Material|SetTextureParameterValue _downsample_mid "RT" RenderTarget)
  (Rendering|Material|SetScalarParameterValue _downsample_mid "Mode" (Math|Conversions|ToFloat(Integer) (Math|Conversions|ToInteger(Byte) Mip%20Mode)) :!class /Script/Engine.MaterialInstanceDynamic)
  (LocalVariables|SetNumMips (- (Math|Float|Round (Math|Float|Log (Math|Conversions|ToFloat(Integer) (Math|Integer|Max(Integer) _sizex _sizey)) 2.0)) 1))
  (LocalVariables|SetCurrent%20Resolution (Utilities|Struct|MakeIntPoint _sizex _sizey))
  ;[0.1979 0.1979 0.1979] Ensure Proper RenderTarget Mips Allocated
  (for _index (range (+ (LocalVariables|GetNumMips) 1))
    (LocalVariables|SetTemp%20RenderTarget (Utilities|Array|Get(acopy) Mip%20RT%20Array _index))
    (bind _output_get (LocalVariables|SetCurrent%20Resolution (Math|IntPoint|IntPoint/Integer (LocalVariables|GetCurrent%20Resolution) 2)))
    (Rendering|CreateAndAssignRenderTarget (LocalVariables|GetTemp%20RenderTarget) (Math|Integer|Max(Integer) (Utilities|Struct|BreakIntPoint _output_get) 1) (Math|Integer|Max(Integer) (.Y (Utilities|Struct|BreakIntPoint _output_get)) 1) "RTF_RGBA16f"
      (:Out
        (Utilities|Array|SetArrayElem Mip%20RT%20Array _index (LocalVariables|GetTemp%20RenderTarget) true))))
  ;[0.0629 0.6207 1] Perform Downsample Render
  (for (_array_element _array_index) Mip%20RT%20Array
    (if (<= _array_index (LocalVariables|GetNumMips))
      (Rendering|DrawMaterialtoRenderTarget _array_element _downsample_mid_2)
      (Rendering|Material|SetTextureParameterValue _downsample_mid_2 "RT" _array_element)
      (else
        (Utilities|Array|RemoveIndex Mip%20RT%20Array _array_index)))))

(fn GenerateSpectrum
  (bind (_maxwaves_5_19925fb1444ed223f41c31961ffbcd57 _seed_6_9215e25244bf277111b6be930635f7b6 _minwavelength_9_ad3f5e67411ab7061f381db7aa2f3b38 _maxwavelength_11_476089b748ef7556964790bedbc4bd3d _wavelengthfalloff_19_2d13de284279a511540f44a3cc36b285 _minamplitude_13_971769f1434b405d434c5b919251e80e _maxamplitude_15_05c75fee4c902830a41b54b05e3bcdab _amplitudefalloff_17_8702450a4204bc3034e389b45cf344ff _dominantwinddirection_32_6799533649c2bffab234bf9b8cbfe51f _directionangularspread_25_74c96e5443d393c08d082d8e1fb45509) (Utilities|Struct|BreakCausticsWaveSpectrumSettings (LocalVariables|GetLocal%20Spectrum)))
  (bind (_maxwaves_5_19925fb1444ed223f41c31961ffbcd57_2 _seed_6_9215e25244bf277111b6be930635f7b6_2 _minwavelength_9_ad3f5e67411ab7061f381db7aa2f3b38_2 _maxwavelength_11_476089b748ef7556964790bedbc4bd3d_2 _wavelengthfalloff_19_2d13de284279a511540f44a3cc36b285_2 _minamplitude_13_971769f1434b405d434c5b919251e80e_2 _maxamplitude_15_05c75fee4c902830a41b54b05e3bcdab_2 _amplitudefalloff_17_8702450a4204bc3034e389b45cf344ff_2 _dominantwinddirection_32_6799533649c2bffab234bf9b8cbfe51f_2 _directionangularspread_25_74c96e5443d393c08d082d8e1fb45509_2 _smallwavesteepness_27_2267de334a87073c4cd1d7b829292dbb _largewavesteepness_29_b325f6fc43e2836ef6bf9bbbc49b9032 _steepnessfalloff_31_178ac6b94552d04f39d5b49301de6d76) (Utilities|Struct|BreakCausticsWaveSpectrumSettings (LocalVariables|GetLocal%20Spectrum)))
  (bind (_maxwaves_5_19925fb1444ed223f41c31961ffbcd57_3 _seed_6_9215e25244bf277111b6be930635f7b6_3 _minwavelength_9_ad3f5e67411ab7061f381db7aa2f3b38_3 _maxwavelength_11_476089b748ef7556964790bedbc4bd3d_3 _wavelengthfalloff_19_2d13de284279a511540f44a3cc36b285_3 _minamplitude_13_971769f1434b405d434c5b919251e80e_3 _maxamplitude_15_05c75fee4c902830a41b54b05e3bcdab_3 _amplitudefalloff_17_8702450a4204bc3034e389b45cf344ff_3) (Utilities|Struct|BreakCausticsWaveSpectrumSettings (LocalVariables|GetLocal%20Spectrum)))
  (bind (_maxwaves_5_19925fb1444ed223f41c31961ffbcd57_4 _seed_6_9215e25244bf277111b6be930635f7b6_4 _minwavelength_9_ad3f5e67411ab7061f381db7aa2f3b38_4 _maxwavelength_11_476089b748ef7556964790bedbc4bd3d_4 _wavelengthfalloff_19_2d13de284279a511540f44a3cc36b285_4) (Utilities|Struct|BreakCausticsWaveSpectrumSettings (LocalVariables|GetLocal%20Spectrum)))
  (bind (_maxwaves_5_19925fb1444ed223f41c31961ffbcd57_5 _seed_6_9215e25244bf277111b6be930635f7b6_5 _minwavelength_9_ad3f5e67411ab7061f381db7aa2f3b38_5 _maxwavelength_11_476089b748ef7556964790bedbc4bd3d_5 _wavelengthfalloff_19_2d13de284279a511540f44a3cc36b285_5 _minamplitude_13_971769f1434b405d434c5b919251e80e_4 _maxamplitude_15_05c75fee4c902830a41b54b05e3bcdab_4 _amplitudefalloff_17_8702450a4204bc3034e389b45cf344ff_4 _dominantwinddirection_32_6799533649c2bffab234bf9b8cbfe51f_3) (Utilities|Struct|BreakCausticsWaveSpectrumSettings (LocalVariables|GetLocal%20Spectrum)))
  (bind _returnvalue 32)
  (bind _returnvalue_2 (Math|Conversions|ToFloat(Integer) _returnvalue))
  (LocalVariables|SetLocal%20Spectrum (Utilities|Struct|MakeCausticsWaveSpectrumSettings (Variables|Waves|GetNum%20Waves) 0 (Variables|Waves|Wavelengths|GetMin%20Wavelength) (Variables|Waves|Wavelengths|GetMax%20Wavelength) (Variables|Waves|Wavelengths|GetWavelength%20Falloff) (Variables|Waves|Amplitudes|GetMin%20Amplitude) (Variables|Waves|Amplitudes|GetMax%20Amplitude) (Variables|Waves|Amplitudes|GetAmplitude%20Falloff) (Variables|Waves|GetWind%20Direction) (Variables|Waves|GetAngular%20Spread) (Variables|Waves|Steepness|GetSmale%20Wave%20Steepness) (Variables|Waves|Steepness|GetLarge%20Wave%20Steepness) (Variables|Waves|Steepness|GetSteepness%20Falloff)))
  (Utilities|Array|Clear (LocalVariables|GetOutWaves))
  (for _index (range _returnvalue)
    (LocalVariables|Setlastalpha (Math|Float|Clamp(Float) (+ (- 1 (/ (Math|Conversions|ToFloat(Integer) _index) _returnvalue_2)) (Math|Random|RandomFloatinRangefromStream (Variables|Waves|GetRandom%20Seed) (* (* 0.0 (/ 1 _returnvalue_2)) -1) (* 0.0 (/ 1 _returnvalue_2))))))
    (Utilities|Array|Add (LocalVariables|GetOutWaves) (select (> _index (- (Variables|Waves|GetNum%20Waves) 1)) (Utilities|Struct|MakeGerstnerWave 2e+03 0.001) (Utilities|Struct|MakeGerstnerWave (Math|Float|Lerp _minwavelength_9_ad3f5e67411ab7061f381db7aa2f3b38_4 _maxwavelength_11_476089b748ef7556964790bedbc4bd3d_4 (Math|Float|Power (LocalVariables|Getlastalpha) _wavelengthfalloff_19_2d13de284279a511540f44a3cc36b285_4)) (Math|Float|Max(Float) (Math|Float|Lerp _minamplitude_13_971769f1434b405d434c5b919251e80e_3 _maxamplitude_15_05c75fee4c902830a41b54b05e3bcdab_3 (Math|Float|Power (LocalVariables|Getlastalpha) _amplitudefalloff_17_8702450a4204bc3034e389b45cf344ff_3)) 0.0001) (Math|Float|Lerp _largewavesteepness_29_b325f6fc43e2836ef6bf9bbbc49b9032 _smallwavesteepness_27_2267de334a87073c4cd1d7b829292dbb (Math|Float|Power (/ (Math|Conversions|ToFloat(Integer) _index) _returnvalue_2) _steepnessfalloff_31_178ac6b94552d04f39d5b49301de6d76)) (Math|Conversions|ToVector(Vector2D) (Math|Vector2D|Normalize2D (Math|Conversions|ToVector2D(Vector) (select (== _index 0) (Math|Conversions|ToVector(Vector2D) _dominantwinddirection_32_6799533649c2bffab234bf9b8cbfe51f_3) (Math|Vector|RotateVectorAroundAxis (Math|Conversions|ToVector(Vector2D) _dominantwinddirection_32_6799533649c2bffab234bf9b8cbfe51f_3) (Math|Random|RandomFloatinRangefromStream (Variables|Waves|GetRandom%20Seed) (* _directionangularspread_25_74c96e5443d393c08d082d8e1fb45509 -1) _directionangularspread_25_74c96e5443d393c08d082d8e1fb45509) "0, 0,1.000000")))))))))
  (return (LocalVariables|GetOutWaves)))

(fn GetWaterMaterial
  (return (Variables|Debug|MIDs|GetWaterSurface%202D%20MID)))

(fn NormalizebyMaxLuminance
  (params (Re%20Calc%20Mips :type Boolean :default true))
  (bind _causticmiprts (Variables|Debug|GetCausticMipRTs))
  (bind _returnvalue (Utilities|Array|LastIndex _causticmiprts))
  (bind _output (Utilities|Array|Get(acopy) _causticmiprts _returnvalue))
  (if (or (Variables|PreviewScene|GetUse%20Light%20Function%20Caustics) (not (Variables|OutputandAnimation|GetSave%20as%20HDR)))
    (if Re%20Calc%20Mips
      ;[0.4773 1 0.4267] Downsample to find Largest Texel
      (CallFunction|GenerateMipChain (Variables|RenderTargets|GetCaustics%20RT) (Variables|Debug|GetCausticMipRTs) "NewEnumerator2")
      (else
        _))
    (Rendering|Material|SetTextureParameterValue (Variables|Debug|MIDs|GetPreview%20RenderTarget%20MID) "MipRT" _output)
    (Variables|OutputAndAnimation|SetMax%20Caustics%20Brightness%20Value (Rendering|RenderTargetSampleUVEditorOnly _output "(X=0.500000,Y=0.500000)"))
    (Rendering|Material|SetTextureParameterValue (Variables|Debug|MIDs|GetModulateLuminanceMID) "MipRT" (Utilities|Array|Get(acopy) (Variables|Debug|GetCausticMipRTs) (Utilities|Array|LastIndex (Variables|Debug|GetCausticMipRTs))))
    (Rendering|DrawMaterialtoRenderTarget (Variables|RenderTargets|GetCaustics%20RT) (Variables|Debug|MIDs|GetModulateLuminanceMID))
    (Rendering|Material|SetScalarParameterValue (Variables|Debug|MIDs|GetWaterSurface%202D%20MID) "Caustics Boost" (Utilities|Struct|BreakLinearColor (Variables|OutputAndAnimation|GetMax%20Caustics%20Brightness%20Value)) :!class /Script/Engine.MaterialInstanceDynamic)
    (else
      (Rendering|Material|SetScalarParameterValue (Variables|Debug|MIDs|GetWaterSurface%202D%20MID) "Caustics Boost" 1.0 :!class /Script/Engine.MaterialInstanceDynamic))))

(fn NormalizeCausticsLuminance
  (params (Re%20Calc%20Mips :type Boolean :default true))
  (bind _causticmiprts (Variables|Debug|GetCausticMipRTs))
  (bind _returnvalue (Utilities|Array|LastIndex _causticmiprts))
  (bind _output (Utilities|Array|Get(acopy) _causticmiprts _returnvalue))
  (bind _1x1mipaverage (Variables|Debug|Get1x1MipAverage))
  (bind _downsample_mid (Variables|Debug|MIDs|GetDownsample%20MID))
  (if Re%20Calc%20Mips
    ;[1 0.7426 0.3375] Downsample Average Mip Chain
    (CallFunction|GenerateMipChain (Variables|RenderTargets|GetCaustics%20RT) (Variables|Debug|GetCausticMipRTs))
    (else
      _))
  ;[0.5426 0.7708 1] Normalize to 1.0f
  (Utilities|CreateMID (Variables|Debug|MIDs|GetModulateLuminanceMID) "/Water/Caustics/Materials/Generation/ModulateLuminance.ModulateLuminance"
    (:Out
      (Rendering|Material|SetTextureParameterValue (Variables|Debug|MIDs|GetModulateLuminanceMID) "MipRT" (select Re%20Calc%20Mips (Utilities|Array|Get(acopy) (Variables|Debug|GetCausticMipRTs) (Utilities|Array|LastIndex (Variables|Debug|GetCausticMipRTs))) (Variables|Debug|Get1x1MipAverage)))
      (Rendering|DrawMaterialtoRenderTarget (Variables|RenderTargets|GetCaustics%20RT) (Variables|Debug|MIDs|GetModulateLuminanceMID))
      (Rendering|Material|SetTextureParameterValue (Variables|Debug|MIDs|GetPreview%20RenderTarget%20MID) "MipRT" _output)
      (Variables|OutputAndAnimation|SetCaustics%20Average%20Brightness%20Value (Rendering|RenderTargetSampleUVEditorOnly _output "(X=0.500000,Y=0.500000)"))
      (if Re%20Calc%20Mips
        (Rendering|CreateAndAssignRenderTarget (Variables|Debug|Get1x1MipAverage) 1 1 "RTF_RGBA16f"
          (:Out
            (Rendering|Material|SetScalarParameterValue _downsample_mid "Mode" 0.0 :!class /Script/Engine.MaterialInstanceDynamic)
            (Rendering|Material|SetTextureParameterValue _downsample_mid "RT" (Utilities|Array|Get(acopy) (Variables|Debug|GetCausticMipRTs) (Utilities|Array|LastIndex (Variables|Debug|GetCausticMipRTs))))
            (Rendering|ClearRenderTarget2D _1x1mipaverage)
            (Rendering|DrawMaterialtoRenderTarget _1x1mipaverage (Variables|Debug|MIDs|GetDownsample%20MID))))))))

(fn NormalizeSubUVLuminance
  (bind _1x1mipaverage (Variables|Debug|Get1x1MipAverage))
  (bind _downsample_mid (Variables|Debug|MIDs|GetDownsample%20MID))
  (Utilities|FlowControl|Sequence
    (:then_0
      (CallFunction|GenerateMipChain (Variables|RenderTargets|GetSubUV%20RT) (Variables|Debug|GetSubUVMipRTs))
      ;[1 1 1] Copy Lowest Mip out
      (Utilities|IsValid (Variables|Debug|Get1x1MipAverage)
        (:Is%20Valid
          _)
        (:Is%20Not%20Valid
          (bind _returnvalue (Rendering|CreateRenderTarget2D 1 1))
          (Variables|Debug|Set1x1MipAverage _returnvalue)))
      (Rendering|Material|SetTextureParameterValue _downsample_mid "RT" (Utilities|Array|Get(acopy) (Variables|Debug|GetSubUVMipRTs) (Utilities|Array|LastIndex (Variables|Debug|GetSubUVMipRTs))))
      (Rendering|ClearRenderTarget2D _1x1mipaverage)
      (Rendering|DrawMaterialtoRenderTarget _1x1mipaverage _downsample_mid)
      (Utilities|CreateMID (Variables|Debug|MIDs|GetModulateLuminanceMID) "/Water/Caustics/Materials/Generation/ModulateLuminance.ModulateLuminance"
        (:Out
          (Rendering|Material|SetTextureParameterValue (Variables|Debug|MIDs|GetModulateLuminanceMID) "MipRT" (Variables|Debug|Get1x1MipAverage))
          (Rendering|DrawMaterialtoRenderTarget (Variables|RenderTargets|GetSubUV%20RT) (Variables|Debug|MIDs|GetModulateLuminanceMID)))))
    (:then_1
      ;[1 1 1] For LDR Mode, Divide by Max
      (if (not (Variables|OutputandAnimation|GetSave%20as%20HDR))
        ;;[0.4773 1 0.4267] Downsample to find Largest Texel
        (CallFunction|GenerateMipChain (Variables|RenderTargets|GetSubUV%20RT) (Variables|Debug|GetSubUVMipRTs) "NewEnumerator2")
        (Variables|OutputAndAnimation|SetMax%20Caustics%20Brightness%20Value (Rendering|RenderTargetSampleUVEditorOnly (Utilities|Array|Get(acopy) (Variables|Debug|GetCausticMipRTs) (Utilities|Array|LastIndex (Variables|Debug|GetCausticMipRTs))) "(X=0.500000,Y=0.500000)"))
        (Rendering|Material|SetTextureParameterValue (Variables|Debug|MIDs|GetModulateLuminanceMID) "MipRT" (Utilities|Array|Get(acopy) (Variables|Debug|GetSubUVMipRTs) (Utilities|Array|LastIndex (Variables|Debug|GetSubUVMipRTs))))
        (Rendering|DrawMaterialtoRenderTarget (Variables|RenderTargets|GetSubUV%20RT) (Variables|Debug|MIDs|GetModulateLuminanceMID))
        (Rendering|Material|SetScalarParameterValue (Variables|Debug|MIDs|GetWaterSurface%20SubUV%20MID) "Caustics Boost" (/ 1.0 (Utilities|Struct|BreakLinearColor (Variables|OutputAndAnimation|GetMax%20Caustics%20Brightness%20Value))) :!class /Script/Engine.MaterialInstanceDynamic)))))

(fn RenderWaveData
  (params (Waves :type Array%20of%20Gerstner%20Wave%20Structures))
  (Rendering|ClearRenderTarget2D "/Water/Textures/DefaultWaveSpectrumRT.DefaultWaveSpectrumRT" (Utilities|Struct|MakeLinearColor 0.1 1e-05))
  (bind (_canvas _size _context) (Rendering|BeginDrawCanvastoRenderTarget "/Water/Textures/DefaultWaveSpectrumRT.DefaultWaveSpectrumRT"))
  (Utilities|IsValid _canvas
    (:Is%20Valid
      (for _index (range (+ (Math|Integer|Min(Integer) (Utilities|Array|LastIndex Waves) (- (Variables|Waves|GetNum%20Waves) 1)) 1))
        (Canvas|DrawBox _canvas (Math|Vector2D|MakeVector2D (+ (* (Math|Conversions|ToFloat(Integer) _index) 2.0) 0.5) 0.5) "(X=0.500000,Y=0.500000)" 1.0 (Math|Conversions|ToLinearColor(Vector) (.Direction (Utilities|Struct|BreakGerstnerWave (Utilities|Array|Get(aref) Waves _index)))))
        (Canvas|DrawBox _canvas (+ (Math|Vector2D|MakeVector2D (+ (* (Math|Conversions|ToFloat(Integer) _index) 2.0) 0.5) 0.5) "(X=1.000000,Y=0.000000)") "(X=0.500000,Y=0.500000)" 1.0 (Utilities|Struct|MakeLinearColor (Utilities|Struct|BreakGerstnerWave (Utilities|Array|Get(aref) Waves _index)) (.Amplitude (Utilities|Struct|BreakGerstnerWave (Utilities|Array|Get(aref) Waves _index))) (.Steepness (Utilities|Struct|BreakGerstnerWave (Utilities|Array|Get(aref) Waves _index))))))
      (Rendering|EndDrawCanvastoRenderTarget _context))))

(fn Reset
  (Variables|Default|SetTime%20Accumulator 0.0)
  (Variables|Default|SetCurrent%20Frame 0)
  (Variables|Default|SetCurrentWaterDepth (Variables|Caustics|GetWater%20Depth))
  (Rendering|ClearRenderTarget2D (Variables|RenderTargets|GetSubUV%20RT))
  (Rendering|Material|SetScalarParameterValue (Variables|Debug|MIDs|GetPreview%20RenderTarget%20MID) "ShowSubUV" 0.0 :!class /Script/Engine.MaterialInstanceDynamic))

(fn SaveTexture
  (EditorScripting|Rendering|RenderTarget2DCreateStaticTexture2DEditorOnly (Utilities|Select :Index (Variables|OutputandAnimation|GetOutput%20Mode) :NewEnumerator0 (Variables|RenderTargets|GetCaustics%20RT) :NewEnumerator1 (Variables|RenderTargets|GetSubUV%20RT) :NewEnumerator2 (Variables|RenderTargets|GetSubUV%20RT)) (Utilities|String|Append (Variables|OutputandAnimation|FilePath|GetTexture%20Save%20Path) (Variables|OutputandAnimation|FilePath|GetTexture%20Save%20Name)) (select (Variables|OutputandAnimation|GetSave%20as%20HDR) "TC_HDR" "TC_Default")))

(fn SetLightFunctionCaustics
  (params (Show%20SubUV :type Boolean))
  (bind _max_caustics_brightness_value (Variables|OutputAndAnimation|GetMax%20Caustics%20Brightness%20Value))
  (bind _r (Utilities|Struct|BreakLinearColor _max_caustics_brightness_value))
  (bind (_x _y) (Utilities|Struct|BreakIntPoint (Variables|OutputandAnimation|GetAnimation%20Frames)))
  (bind _watersurface_2d_mid (Variables|Debug|MIDs|GetWaterSurface%202D%20MID))
  (bind _dir_light (Variables|PreviewScene|GetDir%20Light))
  (bind _lightcomponent (Variables|/Script/Engine.Light|GetLightComponent _dir_light))
  ;[1 1 1] This function is always called because it handles resetting the state between both modes
  (if (Variables|PreviewScene|GetUse%20Light%20Function%20Caustics)
    ;[1 1 1] Find a Directional Light if there is not one set
    (Utilities|IsValid (Variables|PreviewScene|GetDir%20Light)
      (:Is%20Valid
        _)
      (:Is%20Not%20Valid
        (bind _outactors (Actor|GetAllActorsOfClass "/Script/Engine.DirectionalLight"))
        (bind _output_get (Variables|PreviewScene|SetDir%20Light (Utilities|Array|Get(acopy) _outactors)))
        (Utilities|IsValid _output_get
          (:Is%20Valid
            _))))
    (Utilities|CreateMID (Variables|Debug|MIDs|GetLightFunctionMID) "/Water/Caustics/Materials/LightFunctions/M_Caustics_LightFunction_Sun.M_Caustics_LightFunction_Sun"
      (:Out
        ;[1 1 1] Setup and Assign Light Function MID to Light
        (Utilities|CreateMID (Variables|Debug|MIDs|GetLightFunction%20SubUV%20MID) "/Water/Caustics/Materials/LightFunctions/M_Caustics_LightFunction_SubUV.M_Caustics_LightFunction_SubUV"
          (:Out
            (Rendering|Components|Light|SetLightFunctionMaterial (Variables|/Script/Engine.Light|GetLightComponent (Variables|PreviewScene|GetDir%20Light)) (select (and Show%20SubUV (Utilities|Enum|Equal(Enum) (Variables|OutputandAnimation|GetOutput%20Mode) "NewEnumerator1")) (Variables|Debug|MIDs|GetLightFunction%20SubUV%20MID) (Variables|Debug|MIDs|GetLightFunctionMID)))
            (Rendering|Material|SetTextureParameterValue (Variables|Debug|MIDs|GetLightFunctionMID) "Caustics" (Variables|RenderTargets|GetCaustics%20RT))
            (Rendering|Material|SetTextureParameterValue (Variables|Debug|MIDs|GetLightFunction%20SubUV%20MID) "Caustics" (Variables|RenderTargets|GetSubUV%20RT))
            (Rendering|Material|SetVectorParameterValue (Variables|Debug|MIDs|GetLightFunction%20SubUV%20MID) "SubUVFrames" (Utilities|Struct|MakeLinearColor (Math|Conversions|ToFloat(Integer) _x) (Math|Conversions|ToFloat(Integer) _y)) :!class /Script/Engine.MaterialInstanceDynamic)
            (Rendering|Material|SetScalarParameterValue (Variables|Debug|MIDs|GetLightFunctionMID) "Caustics Multiplier" _r :!class /Script/Engine.MaterialInstanceDynamic)
            (Rendering|Material|SetScalarParameterValue "/Water/Materials/MPC/MPC_Water.MPC_Water" "CausticsMaxIntensity" _r :!class /Script/Engine.KismetMaterialLibrary)
            ;[1 1 1] Apply corrections to Intensity and Disabled Brightness to counteract Normalization by Max Brightness
            (Rendering|Components|Light|SetIntensity _lightcomponent (* (Variables|PreviewScene|GetDirectional%20Light%20Initial%20Brightness) (Utilities|Struct|BreakLinearColor (Variables|OutputAndAnimation|GetMax%20Caustics%20Brightness%20Value))))
            (Rendering|Components|Light|SetLightFunctionDisabledBrightness _lightcomponent (/ 1.0 (Utilities|Struct|BreakLinearColor (Variables|OutputAndAnimation|GetMax%20Caustics%20Brightness%20Value))))
            (Rendering|Material|SetScalarParameterValue (Variables|Debug|MIDs|GetWaterSurface%202D%20MID) "Caustics Alpha BP" 0.0 :!class /Script/Engine.MaterialInstanceDynamic)
            (Rendering|Material|SetVectorParameterValue "/Water/Materials/MPC/MPC_Water.MPC_Water" "SunLightVector" (Math|Conversions|ToLinearColor(Vector) (Transformation|GetForwardVector (Variables|/Script/Engine.Light|GetLightComponent (Variables|PreviewScene|GetDir%20Light)))) :!class /Script/Engine.KismetMaterialLibrary)
            (bind _cameralocation (Rendering|GetCursorWorldRay :!class /Script/LandmassEditor.LandmassBlueprintFunctionLibrary))
            (Rendering|Material|SetVectorParameterValue "/Water/Materials/MPC/MPC_Water.MPC_Water" "EditorCameraLocation" (Math|Conversions|ToLinearColor(Vector) _cameralocation) :!class /Script/Engine.KismetMaterialLibrary)))))
    (else
      ;[1 1 1] No LF: Disable all Lightfunction based corrections
      (Utilities|IsValid (Variables|PreviewScene|GetDir%20Light)
        (:Is%20Valid
          (Rendering|Components|Light|SetLightFunctionMaterial (Variables|/Script/Engine.Light|GetLightComponent (Variables|PreviewScene|GetDir%20Light)) 0) ; Remove Lightfunction Material from Dir Light
          (Rendering|Components|Light|SetIntensity (Variables|/Script/Engine.Light|GetLightComponent (Variables|PreviewScene|GetDir%20Light)) (Variables|PreviewScene|GetDirectional%20Light%20Initial%20Brightness)))
        (:Is%20Not%20Valid
          _))
      (Rendering|Material|SetScalarParameterValue _watersurface_2d_mid "Caustics Alpha BP" 1.0 :!class /Script/Engine.MaterialInstanceDynamic) ; Re enable Water Material based Caustics
      (Rendering|Material|SetScalarParameterValue _watersurface_2d_mid "Caustics Multiplier" 1.0 :!class /Script/Engine.MaterialInstanceDynamic)
      (Rendering|Material|SetScalarParameterValue "/Water/Materials/MPC/MPC_Water.MPC_Water" "CausticsMaxIntensity" 1.0 :!class /Script/Engine.KismetMaterialLibrary))))

(fn SetupMIDs
  (bind (_x _y) (Utilities|Struct|BreakIntPoint (Variables|OutputandAnimation|GetAnimation%20Frames)))
  (bind _copy_subuv_frame_mid (Variables|Debug|MIDs|GetCopy%20SubUV%20Frame%20MID))
  (bind _caustics_scatter_mid_singleframe (Variables|Debug|MIDs|GetCaustics%20Scatter%20MID%20SingleFrame))
  (bind _caustics_scatter_mid_singleframe_2 (Variables|Debug|MIDs|GetCaustics%20Scatter%20MID%20SingleFrame))
  (bind _sea_bed_material (Variables|PreviewScene|GetSea%20Bed%20Material))
  (Utilities|FlowControl|Sequence
    (:then_0
      (Utilities|CreateMID (Variables|Debug|MIDs|GetCaustics%20Scatter%20MID%20SingleFrame) (Variables|Default|GetCaustics%20Mesh%20Frame%20Mat)
        (:Out
          (Utilities|CreateMID (Variables|Debug|MIDs|GetCaustics%20Scatter%20MID%20Animation) (Variables|Default|GetCaustics%20Mesh%20Animation%20Mat)
            (:Out
              (Utilities|CreateMID (Variables|Debug|MIDs|GetCopy%20SubUV%20Frame%20MID) "/Water/Caustics/Materials/Generation/Caustics_SubUV_Frame_Copy.Caustics_SubUV_Frame_Copy"
                (:Out
                  (Utilities|CreateMID (Variables|Debug|MIDs|GetPreview%20RenderTarget%20MID) "/Water/Caustics/Materials/Generation/CausticsPreview.CausticsPreview"
                    (:Out
                      (Utilities|CreateMID (Variables|Debug|MIDs|GetWaterSurface%202D%20MID) (select (Variables|PreviewScene|GetUse%20Light%20Function%20Caustics) "/Water/Caustics/Materials/Preview/WaterCausticsPreview_NoCaustics.WaterCausticsPreview_NoCaustics" (select (Variables|PreviewScene|GetShow%20Analytical%20Caustics) "/Water/Caustics/Materials/Preview/WaterCausticsPreview_Analytical.WaterCausticsPreview_Analytical" "/Water/Caustics/Materials/Preview/WaterCausticsPreview_Texture.WaterCausticsPreview_Texture"))
                        (:Out
                          (Utilities|CreateMID (Variables|Debug|MIDs|GetWaterSurface%20SubUV%20MID) "/Water/Caustics/Materials/Preview/WaterCausticsPreview_SubUV.WaterCausticsPreview_SubUV"
                            (:Out
                              (Utilities|CreateMID (Variables|Debug|MIDs|GetRender%20Normal%20MID) "/Water/Caustics/Materials/Generation/RenderTilingWaterNormal.RenderTilingWaterNormal"
                                (:Out
                                  (Utilities|CreateMID (Variables|Debug|MIDs|GetBaseFloorMeshMID) _sea_bed_material
                                    (:Out
                                      (Utilities|CreateMID (Variables|Debug|MIDs|GetSeaBedMID) _sea_bed_material
                                        (:Out
                                          (Utilities|CreateMID (Variables|Debug|MIDs|GetDownsample%20MID) "/Water/Caustics/Materials/Generation/DownsampleRT.DownsampleRT"
                                            (:Out
                                              (Utilities|CreateMID (Variables|Debug|MIDs|GetRemoveOverscanMID) "/Water/Caustics/Materials/Generation/ExtractTilingTextureFromOverscan.ExtractTilingTextureFromOverscan"))))))))))))))))))))))
    (:then_1
      (Rendering|Material|SetScalarParameterValue _caustics_scatter_mid_singleframe "Tiles" (Math|Conversions|ToFloat(Integer) (Variables|Debug|GetTiles)) :!class /Script/Engine.MaterialInstanceDynamic)
      (Rendering|Material|SetScalarParameterValue _caustics_scatter_mid_singleframe "WorldSize" (Variables|Caustics|GetTexture%20World%20Size) :!class /Script/Engine.MaterialInstanceDynamic)
      (Rendering|Material|SetScalarParameterValue _caustics_scatter_mid_singleframe "WaterDepth" (Variables|Caustics|GetWater%20Depth) :!class /Script/Engine.MaterialInstanceDynamic)
      (Rendering|Material|SetScalarParameterValue _caustics_scatter_mid_singleframe_2 "Chromatic Spread" (Variables|Caustics|GetChromatic%20Spread) :!class /Script/Engine.MaterialInstanceDynamic)
      (Rendering|Material|SetScalarParameterValue _caustics_scatter_mid_singleframe_2 "Photon Scale" (* (Variables|Caustics|GetPhoton%20Scale) 4.0) :!class /Script/Engine.MaterialInstanceDynamic)
      (Rendering|Material|SetScalarParameterValue (Variables|Debug|MIDs|GetWaterSurface%202D%20MID) "Caustics Scale" (Variables|Caustics|GetTexture%20World%20Size) :!class /Script/Engine.MaterialInstanceDynamic)
      (Rendering|Material|SetTextureParameterValue (Variables|Debug|MIDs|GetWaterSurface%202D%20MID) "Caustics" (Variables|RenderTargets|GetCaustics%20RT))
      (Rendering|Material|SetScalarParameterValue (Variables|Debug|MIDs|GetBaseFloorMeshMID) "WaterDepth" (Variables|PreviewScene|GetPreview%20Mesh%20Depth) :!class /Script/Engine.MaterialInstanceDynamic)
      (Rendering|Material|SetVectorParameterValue (Variables|Debug|MIDs|GetPreview%20RenderTarget%20MID) "SubUVFrames" (Math|Color|MakeColor (Math|Conversions|ToFloat(Integer) _x) (Math|Conversions|ToFloat(Integer) _y)) :!class /Script/Engine.MaterialInstanceDynamic)
      (Rendering|Material|SetScalarParameterValue (Variables|Debug|MIDs|GetPreview%20RenderTarget%20MID) "ShowSubUV" 0.0 :!class /Script/Engine.MaterialInstanceDynamic)
      (Rendering|Material|SetScalarParameterValue (Variables|Debug|MIDs|GetBaseFloorMeshMID) "EnableDepthVS" 1.0 :!class /Script/Engine.MaterialInstanceDynamic)
      (Rendering|Material|SetScalarParameterValue (Variables|Debug|MIDs|GetSeaBedMID) "EnableDepthVS" 0.0 :!class /Script/Engine.MaterialInstanceDynamic)
      (Rendering|Material|SetScalarParameterValue (Variables|Debug|MIDs|GetPreview%20RenderTarget%20MID) "UseClampedOutput" (Math|Conversions|ToFloat(Boolean) (Variables|PreviewScene|GetUse%20Light%20Function%20Caustics)) :!class /Script/Engine.MaterialInstanceDynamic))
    (:then_2
      (Rendering|Material|SetMaterial (Variables|Default|GetCausticParticleGrid) 0 (Variables|Debug|MIDs|GetCaustics%20Scatter%20MID%20SingleFrame))
      (Rendering|Material|SetMaterial (Variables|Default|GetWaterPreviewGrid) 0 (CallFunction|GetWaterMaterial))
      (Rendering|Material|SetMaterial (Variables|Default|GetWater%20Mesh%20Base) 0 (Variables|Debug|MIDs|GetBaseFloorMeshMID))
      (Rendering|Material|SetScalarParameterValue _copy_subuv_frame_mid "Resolution" (Math|Conversions|ToFloat(Integer) (Variables|OutputandAnimation|GetCaustic%20Mesh%20Resolution)) :!class /Script/Engine.MaterialInstanceDynamic)
      (Rendering|Material|SetScalarParameterValue _copy_subuv_frame_mid "Overscan Quads" (Math|Conversions|ToFloat(Integer) (Variables|Caustics|GetOverscan%20Quads)) :!class /Script/Engine.MaterialInstanceDynamic))
    (:then_3
      (Rendering|Material|SetScalarParameterValue "/Water/Materials/MPC/MPC_Water.MPC_Water" "FreezeTime" (Math|Conversions|ToFloat(Boolean) (Variables|PreviewScene|GetFreeze%20Time)) :!class /Script/Engine.KismetMaterialLibrary)
      (Rendering|Material|SetScalarParameterValue "/Water/Materials/MPC/MPC_Water.MPC_Water" "ManualTime" (Variables|PreviewScene|GetManual%20Frozen%20Time) :!class /Script/Engine.KismetMaterialLibrary))))

(fn StopRendering
  (Default|SetEditorTickEnabled false)
  (Variables|Debug|SetTicking false)
  (Rendering|Material|SetScalarParameterValue "/Water/Materials/MPC/MPC_Water.MPC_Water" "FreezeTime" (Math|Conversions|ToFloat(Boolean) (Variables|PreviewScene|GetFreeze%20Time)) :!class /Script/Engine.KismetMaterialLibrary))

(fn UpdateCapture
  (Rendering|SetVisibility (Variables|Default|GetCausticParticleGrid) true false)
  (Rendering|SceneCapture|CaptureScene (Variables|Default|GetSceneCaptureComponent2D) :!class /Script/Engine.SceneCaptureComponent2D)
  (Rendering|SetVisibility (Variables|Default|GetCausticParticleGrid) (not (Variables|PreviewScene|GetShow%20Water%20Scene)) false)
  (Rendering|DrawMaterialtoRenderTarget (Variables|RenderTargets|GetCaustics%20RT) (Variables|Debug|MIDs|GetRemoveOverscanMID)))

(fn ConstructionScript
  (bind (_x _y) (Utilities|Struct|BreakIntPoint (Variables|OutputandAnimation|GetAnimation%20Frames)))
  (bind _texture_world_size (Variables|Caustics|GetTexture%20World%20Size))
  (bind _returnvalue (/ _texture_world_size 256.0))
  (bind _returnvalue_2 (Transformation|GetActorLocation))
  (bind _texture_world_size_2 (Variables|Caustics|GetTexture%20World%20Size))
  (bind _returnvalue_3 (* _returnvalue 32.0))
  (bind _returnvalue_4 (+ _returnvalue_3 _texture_world_size_2))
  (bind _returnvalue_5 (* "-1.000000,0.000000, 0" _returnvalue_4))
  (bind _texture_world_size_3 (Variables|Caustics|GetTexture%20World%20Size))
  (bind _returnvalue_6 (* "0.000000,0.650000, 0" _texture_world_size_3))
  (bind _show_water_scene (Variables|PreviewScene|GetShow%20Water%20Scene))
  (bind _returnvalue_7 (select _show_water_scene "0, 0,64.000000" "0, 0, 0"))
  (bind _returnvalue_8 (+ _returnvalue_2 _returnvalue_5))
  (bind _returnvalue_9 (+ _returnvalue_8 _returnvalue_7))
  (bind _scenecapturecomponent2d (Variables|Default|GetSceneCaptureComponent2D))
  (bind _result_texture_mesh_plane (Variables|Default|GetResult%20Texture%20Mesh%20Plane))
  (bind _water_mesh_base (Variables|Default|GetWater%20Mesh%20Base))
  (bind _output_get (Variables|OutputandAnimation|SetCaustic%20Mesh%20Resolution (Math|Float|Truncate (Math|Float|Power 2.0 (Math|Conversions|ToFloat(Integer) (Math|Float|Truncate (Math|Float|Log (Math|Conversions|ToFloat(Integer) (Math|Integer|Clamp(Integer) (Variables|OutputandAnimation|GetCaustic%20Mesh%20Resolution) 128 8192)) 2.0)))))))
  (Variables|Debug|SetTiles (/ _output_get 128))
  (Variables|Debug|SetAnimationFrames (Math|Integer|SelectInt (* _x _y) 0 (Utilities|Enum|NotEqual(Enum) (Variables|OutputandAnimation|GetOutput%20Mode) "NewEnumerator0")))
  (Variables|Caustics|SetOverscan%20Quads (Math|Integer|Clamp(Integer) (Variables|Caustics|GetOverscan%20Quads) 0 128))
  (Variables|Debug|SetQuad%20Size (/ (Variables|Caustics|GetTexture%20World%20Size) (Math|Conversions|ToFloat(Integer) (Variables|OutputandAnimation|GetCaustic%20Mesh%20Resolution))))
  (CallFunction|SetupMIDs)
  (CallFunction|AllocatePreviewCaptureRT)
  (Variables|/Script/Engine.SceneCaptureComponent2D|SetOrthoWidth (+ (Variables|Caustics|GetTexture%20World%20Size) (* (* (Variables|Debug|GetQuad%20Size) (Math|Conversions|ToFloat(Integer) (Variables|Caustics|GetOverscan%20Quads))) 2.0)) _scenecapturecomponent2d)
  (Rendering|SceneCapture|ShowOnlyComponent _scenecapturecomponent2d (Variables|Default|GetCausticParticleGrid))
  (CallFunction|AddPhotonMeshes)
  ;[0.1354 0.1354 0.1354] RT Preview
  (Rendering|SetVisibility (Variables|Default|GetResult%20Texture%20Mesh%20Plane) (Variables|PreviewScene|GetShow%20Texture%20Plane) false)
  (Rendering|SetVisibility (Variables|Default|GetWater%20Mesh%20Base) (Variables|PreviewScene|GetShow%20Water%20Scene) false)
  (Transformation|SetWorldTransform _result_texture_mesh_plane (Math|Transform|MakeTransform _returnvalue_9 "0, 0, 0" (Math|Conversions|ToVector(Float) _texture_world_size_2)) false false)
  (Transformation|SetWorldLocation (Variables|Default|GetRenderTarget%20Preview) (+ _returnvalue_9 _returnvalue_6) false false)
  (Transformation|SetWorldLocation (Variables|Default|GetPhoton%20HISMC%20Grid) (+ _returnvalue_6 (select (Variables|PreviewScene|GetShow%20Water%20Scene) "0, 0,72.000000" "0, 0, 0")) false false)
  (Rendering|Components|TextRender|SetText (Variables|Default|GetPhoton%20HISMC%20Grid) (select (Variables|PreviewScene|GetShow%20Water%20Scene) (Utilities|Text|MakeLiteralText "NSLOCTEXT(\"[A4C6A3D94C07C3FC7B8E29AFC5C6FC52]\", \"C93209854B5C149172F01AB84CA5966B\", \"Result Applied to Water Material\")") (Utilities|Text|MakeLiteralText "NSLOCTEXT(\"[A4C6A3D94C07C3FC7B8E29AFC5C6FC52]\", \"898F26AE473C45018788E08D3B885069\", \"Caustic Photons to Render\")")))
  (Rendering|Material|SetMaterial _result_texture_mesh_plane 0 (Variables|Debug|MIDs|GetPreview%20RenderTarget%20MID))
  ;[0.3217 0.4631 1] Preview Water and Base
  (Transformation|SetRelativeScale3D _water_mesh_base (Math|Conversions|ToVector(Float) (* (/ (Variables|Caustics|GetTexture%20World%20Size) 256.0) (Variables|PreviewScene|GetPreview%20Scene%20XY%20Scale))))
  (Components|StaticMesh|SetStaticMesh _water_mesh_base (Utilities|Select :Index (Variables|PreviewScene|GetPreview%20Scene) :NewEnumerator0 "/Water/Caustics/Meshes/CausticsPreviewBase.CausticsPreviewBase" :NewEnumerator1 "/Water/Caustics/Meshes/CausticsPreviewIsland.CausticsPreviewIsland"))
  (Rendering|SetVisibility (Variables|Default|GetWater%20Mesh%20Base) (Variables|PreviewScene|GetShow%20Water%20Scene) false)
  (CallFunction|AddWaterPreviewMeshes)
  (if (Variables|Waves|GetGenerate)
    ;[1 1 1] Make Gerstner Waves
    (bind _waves (CallFunction|GenerateSpectrum))
    (bind _output_get_2 (Variables|Default|SetWaves _waves))
    (CallFunction|RenderWaveData _output_get_2)
    (else
      _))
  (Rendering|SetVisibility (Variables|Default|GetCausticParticleGrid) (not (Variables|PreviewScene|GetShow%20Water%20Scene)) false)
  (Variables|Default|SetCurrentWaterDepth (Variables|Caustics|GetWater%20Depth))
  (Variables|Default|SetDepthPerFrame (/ (- (Variables|Caustics|GetWater%20Depth%20End) (Variables|Caustics|GetWater%20Depth)) (Math|Float|Max(Float) (Math|Conversions|ToFloat(Integer) (Variables|Debug|GetAnimationFrames)) 1.0)))
  (CallFunction|TickforUpdate))
