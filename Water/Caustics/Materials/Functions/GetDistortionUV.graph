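# Water material function that approximates the engine's refraction-distorted ViewportUV, used in deriving world position behind water for caustics.
# A Custom HLSL node scales the view-space difference between world up and the pixel normal by refraction minus one.
# That offset is multiplied by a parameter and a fixed axis-scale vector, then added to ViewportUV.
# Distortion, fade and depth inputs feed the node but leave its result unaffected.
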MaterialFunction /Water/Caustics/Materials/Functions/GetDistortionUV
graph {
 properties {
  bExposeToLibrary=true
  description="This function attempts to match the intenral distorted ViewportUV that Water uses, taking into account refraction. It is used as part of the process to derive the world position behind water."
  libraryCategoriesText [Water]
 }
 nodes [
  MaterialExpressionFunctionInput_1 = FunctionInput(Preview=PixelNormalWS()) {bUsePreviewValueAsDefault=true, inputName=PixelNormalWS}
  MaterialExpressionScreenPosition_1 = ScreenPosition()
  n1 = FunctionInput() {bUsePreviewValueAsDefault=true, inputName=Refraction, inputType=FunctionInput_Scalar, previewValue=<1.05 0 0 1>}
  MaterialExpressionAppendVector_6 = AppendVector(A=@n1, B=Constant())
  n2 = Divide(B=ViewProperty() {property=MEVP_TanHalfFieldOfView}.Property)
  n3 = ViewProperty() {property=MEVP_ViewSize}
  n4 = MaterialFunctionCall(Float2=@n3.Property) {materialFunction=&BreakOutFloat2Components}
  MaterialExpressionAppendVector_5 = AppendVector(A=AppendVector(A=ComponentMask(None=@n2) {r=true}, B=Divide(A=@n4.R, B=@n4.R)), B=ComponentMask(None=@n3.Property) {g=true, r=true})
  MaterialExpressionScalarParameter_14 = ScalarParameter() {defaultValue=1, parameterName=FD}
  MaterialExpressionMultiply_7 = Multiply(A=SceneDepthWithoutWater!(UVs).R)
  n5 = Custom!(
   Normal=@MaterialExpressionFunctionInput_1
   ViewUV=@MaterialExpressionScreenPosition_1.ViewportUV
   Refraction=@MaterialExpressionAppendVector_6
   DistortionParameters=@MaterialExpressionAppendVector_5
   FadeDist=@MaterialExpressionScalarParameter_14
   noop=@MaterialExpressionMultiply_7
  ) {
   code="##define SINGLE_LAYER_WATER_DEPTH_SCALE 100.0f\r\n\r\n\tfloat2 WaterNDC = Parameters.ScreenPosition.xy / Parameters.ScreenPosition.w;\r\n\tfloat2 ViewportUV = WaterNDC * ResolvedView.ScreenPositionScaleBias.xy + ResolvedView.ScreenPositionScaleBias.wz;\r\n\r\n\r\nfloat WaterDepth = ConvertFromDeviceZ(Parameters.SvPosition.z);\r\nfloat PixelSceneDepth = MaterialExpressionSceneDepthWithoutWater(ViewportUV, 0).x;\r\n\r\nfloat3 ViewNormal = normalize(TransformWorldVectorToView(Normal));\r\nfloat3 ViewVertexNormal  = normalize(TransformWorldVectorToView(float3(0, 0, 1)));\r\n\r\n\t\tfloat2 ViewportUVDistortion = (ViewVertexNormal.xy - ViewNormal.xy) * (Refraction.x - 1.0);\r\n\t\r\n\t\r\n\treturn ViewportUVDistortion;"
   description=Get Distorted UV
   outputType=CMOT_Float2
   inputs [
    {inputName=Normal, input {expression=&self:MaterialExpressionFunctionInput_1}}
    {inputName=ViewUV, input {expression=&self:MaterialExpressionScreenPosition_1}}
    {inputName=Refraction, input {expression=&self:MaterialExpressionAppendVector_6}}
    {inputName=DistortionParameters, input {expression=&self:MaterialExpressionAppendVector_5}}
    {inputName=FadeDist, input {expression=&self:MaterialExpressionScalarParameter_14}}
    {inputName=noop, input {expression=&self:MaterialExpressionMultiply_7}}
   ]
  }
  n6 = Multiply(A=ScalarParameter() {defaultValue=1, parameterName=Multiply Refraction}, B=Constant2Vector() {g=-2, r=1})
  FunctionOutput(None=Add(A=ScreenPosition().ViewportUV, B=Multiply(A=@n5, B=@n6)))
  MaterialFunctionCall(Float2=@n3.Property) {materialFunction=&BreakOutFloat2Components}
 ]
}
