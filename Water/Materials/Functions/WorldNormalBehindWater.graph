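# Material function that reconstructs the world-space normal of the scene behind SingleLayerWater from water-free scene depth.
# It samples SceneDepthWithoutWater at the viewport UV and one texel right and down, with texel size scaled by the RefractionDownsampleFactor parameter.
# Custom HLSL builds tangent and binormal from depth deltas, camera vector and FOV, then normalizes their cross product.
# Quality is poor unless r.Water.SingleLayer.RefractionFullPrecision is 1.
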MaterialFunction /Water/Materials/Functions/WorldNormalBehindWater
graph {
 properties {
  bExposeToLibrary=true
  description=Computes the WorldNormal behind SingleLayerWater using SceneDepthBehindWater. Note quality will be poor unless r.Water.SingleLayer.RefractionFullPrecision is set to 1.
  libraryCategoriesText [Water]
 }
 nodes [
  MaterialExpressionMultiply_10 = Multiply(A=CameraVectorWS()) {constB=-1}
  MaterialExpressionDivide_1 = Divide(B=ViewProperty() {property=MEVP_ViewSize}.Property)
  MaterialExpressionViewProperty_0 = ViewProperty() {property=MEVP_TanHalfFieldOfView}
  n1 = FunctionInput(Preview=ScreenPosition().ViewportUV) {bUsePreviewValueAsDefault=true, inputName=ViewportUV, inputType=FunctionInput_Vector2}
  MaterialExpressionSceneDepthWithoutWater_0 = SceneDepthWithoutWater!(UVs=@n1)
  n2 = Divide(B=Divide(A=ViewProperty() {property=MEVP_ViewSize}.Property, B=ScalarParameter() {defaultValue=2, parameterName=RefractionDownsampleFactor}))
  n3 = Constant2Vector() {r=1}
  MaterialExpressionSceneDepthWithoutWater_1 = SceneDepthWithoutWater(UVs=Add(A=@n1, B=Multiply(A=@n2, B=@n3)))
  n4 = Constant2Vector() {g=1}
  MaterialExpressionSceneDepthWithoutWater_2 = SceneDepthWithoutWater(UVs=Add(A=@n1, B=Multiply(A=@n2, B=@n4)))
  n5 = Custom!(
   CamVec=@MaterialExpressionMultiply_10
   SceneTexelSize=@MaterialExpressionDivide_1
   TanHalfFOV=@MaterialExpressionViewProperty_0.Property
   depth_c=@MaterialExpressionSceneDepthWithoutWater_0.R
   depth_x=@MaterialExpressionSceneDepthWithoutWater_1.R
   depth_y=@MaterialExpressionSceneDepthWithoutWater_2.R
  ) {
   code="float3 camvec = -CamVec;\r\n\r\nfloat camdot = dot(camvec, ResolvedView.ViewForward);\r\ncamvec/=camdot;\r\n\r\nfloat2 texeltan = TanHalfFOV * (SceneTexelSize * 2);\r\nfloat2 texelinvcos = sqrt((texeltan*texeltan) + 1);\r\n\r\nfloat2 viewUV = GetViewportUV(Parameters);\r\nfloat2 bufferUV = ViewportUVToBufferUV(viewUV);\r\nfloat2 screenUV = min(max(bufferUV,ResolvedView.BufferBilinearUVMinMax.xy),ResolvedView.BufferBilinearUVMinMax.zw);\r\n\r\n//float depth_c = CalcSceneDepth(screenUV);\r\n//float depth_x = CalcSceneDepth(screenUV + float2(SceneTexelSize.x, 0));\r\n//float depth_y = CalcSceneDepth(screenUV + float2(0, SceneTexelSize.y));\r\n\r\nTanHalfFOV *= SceneTexelSize * 2 * depth_c;\r\n\r\nfloat3 tangent = (depth_x-depth_c) * camvec * texelinvcos.x;\r\ntangent += TanHalfFOV.x * ResolvedView.ViewRight;\r\n\r\nfloat3 binormal = (depth_y - depth_c) * camvec * texelinvcos.y;\r\nbinormal += TanHalfFOV.y * -ResolvedView.ViewUp;\r\n\r\nreturn normalize(cross(tangent, binormal));"
   description=Normal From Depth
   inputs [
    {inputName=CamVec, input {expression=&self:MaterialExpressionMultiply_10}}
    {inputName=SceneTexelSize, input {expression=&self:MaterialExpressionDivide_1}}
    {inputName=TanHalfFOV, input {expression=&self:MaterialExpressionViewProperty_0}}
    {
     inputName=depth_c
     input {expression=&self:MaterialExpressionSceneDepthWithoutWater_0, mask=1, maskR=1}
    }
    {
     inputName=depth_x
     input {expression=&self:MaterialExpressionSceneDepthWithoutWater_1, mask=1, maskR=1}
    }
    {
     inputName=depth_y
     input {expression=&self:MaterialExpressionSceneDepthWithoutWater_2, mask=1, maskR=1}
    }
   ]
  }
  FunctionOutput(None=@n5) {outputName=WorldNormal}
 ]
}
