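# Material function that projects caustics for Single Layer Water, indexing a sub-UV flipbook as a pseudo-volume by water depth instead of time.
# Refraction-distorted world positions behind the water give light-parallaxed UVs and depth, measured as negative world Z.
# A dithered custom HLSL raymarch along view and light directions adds a weighted volume blur to the sampled frame.
# Scene-depth fade and normal-dot-light mask the brightened result, also exposing the blur separately.
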MaterialFunction /Water/Caustics/Materials/Functions/WaterCaustics_SubUVAnimation_Depth
graph {
 properties {
  bExposeToLibrary=true
  description="This function will display Caustics on a SingleLayerWater material. It uses a volume texture to render. Instead of using time so indez the Z dimension, water depth will be used.  Create a material instance and adjust the param 'Max Encoded Depth' to scale the depth."
  libraryCategoriesText [Water]
 }
 nodes [
  ComponentMask(None=ObjectPositionWS()) {b=true}
  MaterialExpressionTextureObjectParameter_0 = TextureObjectParameter() {parameterName=Caustics, samplerType=SAMPLERTYPE_LinearColor, texture=null}
  n1 = SkyAtmosphereLightDirection()
  n2 = Divide(A=ComponentMask(None=@n1) {g=true, r=true}, B=LinearInterpolate(B=ComponentMask(None=@n1) {b=true}) {constA=1, constAlpha=0.4})
  n3 = Multiply(A=@n2, B=ScalarParameter() {defaultValue=1, parameterName=Light Parallax})
  n4 = FunctionInput() {bUsePreviewValueAsDefault=true, inputName=Refraction, inputType=FunctionInput_Scalar, previewValue=<1.05 0 0 1>}
  n5 = MaterialFunctionCall!(Refraction=@n4) {materialFunction=&/GetDistortionUV}
  n6 = MaterialFunctionCall!(ScreenUV=@n5.Result) {materialFunction=/Water/Materials/Functions/WorldPositionBehindWater}
  n7 = Subtract(B=ComponentMask(None=@n6.Result) {b=true}) {constA=0}
  n8 = FunctionInput() {bUsePreviewValueAsDefault=true, inputName=Tiling, inputType=FunctionInput_Scalar, sortPriority=2, previewValue=<1024 0 0 1>}
  MaterialExpressionDivide_6 = Divide(A=Add(A=Multiply(A=@n3, B=@n7), B=ComponentMask(None=@n6.Result) {g=true, r=true}), B=@n8)
  n9 = Transform(None=CameraVectorWS()) {transformSourceType=TRANSFORMSOURCE_World, transformType=TRANSFORM_Tangent}
  n10 = Divide(A=ComponentMask(None=@n9) {g=true, r=true}, B=ComponentMask(None=@n9) {b=true})
  n11 = ScalarParameter() {defaultValue=0.5, parameterName=RayDirMult}
  MaterialExpressionComponentMask_8 = ComponentMask(None=Add(A=Multiply(A=@n10, B=@n11) {constB=0.5}, B=Multiply(A=@n2) {constB=-0.5})) {g=true, r=true}
  n12 = ScalarParameter() {defaultValue=0.05, parameterName=Depth Limt}
  MaterialExpressionMin_0 = Min(A=Divide(A=Divide(A=@n7, B=@n12)), B=ScalarParameter() {defaultValue=0.1, parameterName=Blur Distance})
  n13 = ScalarParameter() {defaultValue=1, parameterName=Dither Reduction}
  n14 = Saturate(None=MaterialFunctionCall(Random=Constant()) {materialFunction=&DitherTemporalAA}.Result)
  MaterialExpressionReroute_3 = Reroute(None=Multiply(A=ScalarParameter() {defaultValue=32, parameterName=Blur Steps}, B=LinearInterpolate(A=@n13, Alpha=@n14) {constA=0.5}))
  MaterialExpressionConstant_0 = Constant()
  MaterialExpressionComponentMask_0 = ComponentMask(None=VectorParameter() {parameterName=SubUVFrames, defaultValue=<rgba 8 8 0 0>}.RGB) {g=true, r=true}
  MaterialExpressionSaturate_2 = Saturate(None=Divide(A=@n7, B=ScalarParameter() {defaultValue=256, parameterName=Max Encoded Depth}))
  n15 = Custom!(
   Tex=@MaterialExpressionTextureObjectParameter_0
   UVW=@MaterialExpressionDivide_6
   Vector=@MaterialExpressionComponentMask_8
   Distance=@MaterialExpressionMin_0
   Steps=@MaterialExpressionReroute_3
   Mip=@MaterialExpressionConstant_0
   XYFrames=@MaterialExpressionComponentMask_0
   NormalizedDepth=@MaterialExpressionSaturate_2
  ) {
   code="\r\n\r\nfloat4 accum = 0;\r\nfloat3 curUVW = float3(UVW.xy, NormalizedDepth);\r\nconst int numFrames = XYFrames.x*XYFrames.y;\r\n\r\nfloat stepsize = Distance / Steps;\r\n//stepsize /=10.0;\r\nfloat gradstep = 1.0 / Steps;\r\nfloat gradpos = 0;\r\nfloat zstep = NormalizedDepth / Steps;\r\n\r\nfor(int i =0; i < Steps; i++)\r\n{\r\n\r\n\r\n\tgradpos = float(i) * gradstep;\r\n\t//float3 spectrum = saturate(1-abs(2*gradpos- float3(0,1,2)));\r\n\t\r\n\t//float3 caustics = Tex.SampleLevel(TexSampler, curUVW, Mip).rgb;\r\n\tfloat3 caustics = PseudoVolumeTexture(Tex, TexSampler, curUVW, XYFrames, numFrames);\r\n\t\r\n\taccum.rgb += caustics;\r\n\t//accum.a += caustics.r * spectrum.b;\r\n\t\r\n\tcurUVW.xy += Vector.xy * stepsize;\r\n\tcurUVW.z -= zstep;\r\n}\r\n\r\nreturn float4(accum/Steps );"
   description=Volume Blur
   outputType=CMOT_Float4
   inputs [
    {
     inputName=Tex
     input {expression=&self:MaterialExpressionTextureObjectParameter_0}
    }
    {inputName=UVW, input {expression=&self:MaterialExpressionDivide_6}}
    {
     inputName=Vector
     input {expression=&self:MaterialExpressionComponentMask_8}
    }
    {inputName=Distance, input {expression=&self:MaterialExpressionMin_0}}
    {inputName=Steps, input {expression=&self:MaterialExpressionReroute_3}}
    {inputName=Mip, input {expression=&self:MaterialExpressionConstant_0}}
    {
     inputName=XYFrames
     input {expression=&self:MaterialExpressionComponentMask_0}
    }
    {
     inputName=NormalizedDepth
     input {expression=&self:MaterialExpressionSaturate_2}
    }
   ]
  }
  ComponentMask(None=@n15) {a=true}
  n16 = ComponentMask(None=@n15) {b=true, g=true, r=true}
  FunctionOutput(None=@n16) {outputName=Blurred Caustics, sortPriority=1}
  n17 = Divide(
   A=Frac(None=@MaterialExpressionDivide_6)
   B=ComponentMask(None=VectorParameter() {parameterName=SubUVFrames, defaultValue=<rgba 8 8 0 0>}.RGB) {g=true, r=true}
  )
  n18 = VectorParameter() {parameterName=SubUVFrames, defaultValue=<rgba 8 8 0 0>}
  n19 = Subtract(A=Multiply(A=@n18.R, B=@n18.R))
  n20 = Divide(
   A=MaterialFunctionCall!(1D Index=Multiply(A=@MaterialExpressionSaturate_2, B=@n19), Cells XY=@n18.RGB) {materialFunction=/Engine/Functions/Engine_MaterialFunctions01/Coordinates/1Dto2DIndex}.2D Index
   B=ComponentMask(None=@n18.RGB) {g=true, r=true}
  )
  n21 = Add(
   A=Multiply(A=@n16, B=ScalarParameter() {parameterName=Volume Blur})
   B=TextureSampleParameter2D!(UVs=Add(A=@n17, B=@n20), Level=ScalarParameter() {parameterName=MipLevel}, Apply View MipBias) {mipValueMode=TMVM_MipLevel, parameterName=Caustics, samplerType=SAMPLERTYPE_LinearColor, texture=null}.RGB
  )
  n22 = FunctionInput() {bUsePreviewValueAsDefault=true, inputName=Caustics Brightness, inputType=FunctionInput_Scalar, sortPriority=4, previewValue=<8 6 0 1>}
  n23 = SceneDepthWithoutWater!(UVs)
  n24 = PixelDepth()
  n25 = Saturate(None=Divide(A=Subtract(A=@n23.R, B=@n24.R), B=ScalarParameter() {defaultValue=30, parameterName=Caustics Fade In Distance}))
  n26 = Max(A=DotProduct(
    A=MaterialFunctionCall!(ViewportUV=@n5.Result) {materialFunction=/Water/Materials/Functions/WorldNormalBehindWater}.WorldNormal
    B=SkyAtmosphereLightDirection()
   )) {constB=0}
  FunctionOutput(None=LinearInterpolate(B=Multiply(A=@n21, B=@n22), Alpha=Multiply(A=@n25, B=@n26)) {constA=1}) {outputName=Caustics, sortPriority=1}
 ]
}
