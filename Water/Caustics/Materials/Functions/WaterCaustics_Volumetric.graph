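# Material function for SingleLayerWater that renders animated caustics on the scene behind the water from a volume texture.
# Refraction-distorted world XY behind the water, divided by Tiling, forms UV, while panning time drives W as animation.
# Water depth sets log2 mip bias to soften deeper caustics, and fades the brightness-scaled result in from a neutral 1.
# SubUV, ray-stepped blur and normal-versus-light branches are left unconnected to the output.
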MaterialFunction /Water/Caustics/Materials/Functions/WaterCaustics_Volumetric
graph {
 properties {
  bExposeToLibrary=true
  description=This function will display Caustics on a SingleLayerWater material. It uses a volume texture to render.
  libraryCategoriesText [Water]
 }
 nodes [
  n1 = FunctionInput() {bUsePreviewValueAsDefault=true, inputName=Refraction, inputType=FunctionInput_Scalar, sortPriority=1, previewValue=<0.3 0 0 1>}
  n2 = MaterialFunctionCall!(Refraction=@n1) {materialFunction=&/GetDistortionUV}
  n3 = FunctionInput() {bUsePreviewValueAsDefault=true, inputName=Tiling, inputType=FunctionInput_Scalar, sortPriority=2, previewValue=<1024 0 0 1>}
  n4 = Divide(
   A=ComponentMask(None=MaterialFunctionCall!(ScreenUV=@n2.Result) {materialFunction=/Water/Materials/Functions/WorldPositionBehindWater}.Result) {g=true, r=true}
   B=@n3
  )
  n5 = ScalarParameter() {defaultValue=3, parameterName=Caustics Panning 2asdfSpeed}
  ComponentMask(None=MaterialFunctionCall!(
    Texture=TextureObject() {samplerType=SAMPLERTYPE_LinearColor, texture=null}
    UVs=@n4
    SubImages=Constant() {r=4}
    Frame=MaterialFunctionCall(Speed=@n5) {materialFunction=&TimeWithSpeedVariable}.Time
   ) {materialFunction=/Engine/Functions/Engine_MaterialFunctions03/Texturing/SubUV_Function}.RGB) {r=true}
  Constant2Vector()
  MaterialExpressionAppendVector_0 = AppendVector(
   A=@n4
   B=MaterialFunctionCall(Speed=ScalarParameter() {defaultValue=0.5, parameterName=Caustics Panning Speed}) {materialFunction=&TimeWithSpeedVariable}.Frac Time
  )
  MaterialExpressionMultiply_2 = Multiply(
   A=Subtract(A=SceneDepthWithoutWater!(UVs).R, B=PixelDepth().R)
   B=ScalarParameter() {defaultValue=0.0015, parameterName=Caustics Mip Density}
  )
  n6 = Custom!(n=@MaterialExpressionMultiply_2) {
   code="return log2(n);"
   description=log2n
   outputType=CMOT_Float1
   inputs [{inputName=n, input {expression=&self:MaterialExpressionMultiply_2}}]
  }
  MaterialExpressionMax_1 = Max(A=@n6) {constB=0}
  n7 = TextureSample!(
   UVs=@MaterialExpressionAppendVector_0
   Tex=FunctionInput(Preview=TextureObject() {samplerType=SAMPLERTYPE_LinearColor, texture=/Water/Textures/Caustics/Caustics_Animated_01_HDR_Volume}) {bUsePreviewValueAsDefault=true, inputName=Volume Texture, inputType=FunctionInput_VolumeTexture}
   Bias=@MaterialExpressionMax_1
   Apply View MipBias
  ) {mipValueMode=TMVM_MipBias, texture=&Caustics_Animated_01_HDR_Volume}
  n8 = FunctionInput() {bUsePreviewValueAsDefault=true, inputName=Caustics Brightness, inputType=FunctionInput_Scalar, sortPriority=4, previewValue=<1 0 0 1>}
  n9 = SceneDepthWithoutWater()
  n10 = PixelDepth()
  n11 = Saturate(None=Divide(A=Subtract(A=@n9.R, B=@n10.R), B=ScalarParameter() {defaultValue=200, parameterName=Caustics Fade In Distance}))
  FunctionOutput(None=LinearInterpolate(B=Multiply(A=@n7.RGB, B=@n8), Alpha=Multiply(A=@n11)) {constA=1})
  MaterialExpressionTextureObject_0 = TextureObject() {samplerType=SAMPLERTYPE_LinearColor, texture=&Caustics_Animated_01_HDR_Volume}
  n12 = Transform(None=CameraVectorWS()) {transformSourceType=TRANSFORMSOURCE_World, transformType=TRANSFORM_Tangent}
  n13 = ComponentMask(None=@n12) {g=true, r=true}
  n14 = ComponentMask(None=@n12) {b=true}
  MaterialExpressionAdd_0 = Add(
   A=Multiply(A=Divide(A=@n13, B=@n14), B=ScalarParameter() {defaultValue=0.5, parameterName=asdf})
   B=ComponentMask(None=VectorParameter() {parameterName=a, defaultValue=<rgba 1 0 0 0>}.RGB) {g=true, r=true}
  )
  MaterialExpressionScalarParameter_1 = ScalarParameter() {defaultValue=0.5, parameterName=BLurDistance}
  MaterialExpressionScalarParameter_2 = ScalarParameter() {defaultValue=32, parameterName=Steps}
  n15 = Custom!(
   Tex=@MaterialExpressionTextureObject_0
   UVW=@MaterialExpressionAppendVector_0
   Vector=@MaterialExpressionAdd_0
   Distance=@MaterialExpressionScalarParameter_1
   Steps=@MaterialExpressionScalarParameter_2
   Mip=@MaterialExpressionMax_1
  ) {
   code="\r\n\r\nfloat3 accum = 0;\r\nfloat3 curUVW = UVW;\r\n\r\nfloat stepsize = Distance / Steps;\r\nstepsize /=10;\r\n\r\nfor(int i =0; i < Steps; i++)\r\n{\r\n\taccum += Tex.SampleLevel(TexSampler, curUVW, Mip);\r\n\t\r\n\tcurUVW += float3(Vector.xy * stepsize, 0);\r\n}\r\n\r\nreturn accum/Steps;"
   description=Volume Blur
   inputs [
    {inputName=Tex, input {expression=&self:MaterialExpressionTextureObject_0}}
    {inputName=UVW, input {expression=&self:MaterialExpressionAppendVector_0}}
    {inputName=Vector, input {expression=&self:MaterialExpressionAdd_0}}
    {inputName=Distance, input {expression=&self:MaterialExpressionScalarParameter_1}}
    {inputName=Steps, input {expression=&self:MaterialExpressionScalarParameter_2}}
    {inputName=Mip, input {expression=&self:MaterialExpressionMax_1}}
   ]
  }
  LinearInterpolate(A=@n7.RGB, B=@n15, Alpha=ScalarParameter() {defaultValue=0.75, parameterName=Volume Blur})
  n16 = MaterialFunctionCall!(ViewportUV) {materialFunction=/Water/Materials/Functions/WorldNormalBehindWater}
  n17 = SkyAtmosphereLightDirection()
  Reroute(None=Max(A=DotProduct(A=@n16.WorldNormal, B=@n17)) {constB=0})
 ]
}
