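# Water material function that computes a caustics multiplier for surfaces seen through water, combining an animated volume-texture pattern with fluid simulation caustics.
# Refracted behind-water positions are projected along the sky light to the surface for UVs, which simulation normals distort.
# Time picks the volume slice and depth behind the surface sets the mip.
# Exponentially shaped normal-Z differences between offset simulation samples add caustics, with the result masked by sun-facing normals.
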MaterialFunction /Water/Caustics/Materials/Functions/WaterCaustics_Volumetric_Fluidsim
graph {
 properties {
  bExposeToLibrary=true
  description=This is a caustics material function that uses a Volume texture for the base tiling and then injects caustics generated by BP_FluidSim_01.
  libraryCategoriesText [Water]
 }
 nodes [
  Add()
  n1 = SkyAtmosphereLightDirection()
  n2 = ComponentMask(None=@n1) {g=true, r=true}
  n3 = ComponentMask(None=@n1) {b=true}
  n4 = FunctionInput() {bUsePreviewValueAsDefault=true, inputName=Caustic Refraction, inputType=FunctionInput_Scalar, sortPriority=1, previewValue=<0.3 0 0 1>}
  n5 = MaterialFunctionCall!(Refraction=@n4) {materialFunction=&/GetDistortionUV}
  n6 = MaterialFunctionCall!(ScreenUV=@n5.Result) {materialFunction=/Water/Materials/Functions/WorldPositionBehindWater}
  n7 = Subtract(A=ComponentMask(None=WorldPosition().XYZ) {b=true}, B=ComponentMask(None=@n6.Result) {b=true})
  n8 = Max(A=@n7) {constB=0}
  n9 = Add(A=Multiply(A=Divide(A=@n2, B=@n3), B=@n8), B=ComponentMask(None=@n6.Result) {g=true, r=true})
  n10 = MaterialFunctionCall!(WorldPosition=@n9) {materialFunction=/Water/FluidSimulation/Materials/Functions/Fluidsim_WorldPosition_to_UV}
  n11 = TextureSampleParameter2D(UVs=@n10.UV) {
   parameterName=NormalAndHeight
   samplerSource=SSM_Clamp_WorldGroupSettings
   samplerType=SAMPLERTYPE_LinearColor
   texture=/Engine/Functions/Engine_MaterialFunctions02/PivotPainter2/Black_1x1_EXR_Texture
  }
  n12 = Multiply(A=ComponentMask(None=@n11.RGB) {g=true, r=true}, B=ScalarParameter() {defaultValue=-2, parameterName=Fluidsim Caustics Distortion})
  n13 = FunctionInput() {bUsePreviewValueAsDefault=true, inputName=Tiling, inputType=FunctionInput_Scalar, sortPriority=2, previewValue=<1024 0 0 1>}
  n14 = Divide(A=Add(A=@n9, B=Multiply(A=@n8, B=@n12)), B=@n13)
  n15 = MaterialFunctionCall(Speed=ScalarParameter() {defaultValue=0.5, parameterName=Caustics Panning Speed}) {materialFunction=&TimeWithSpeedVariable}
  n16 = TextureObject() {texture=/Water/Textures/Caustics/Caustics_Animated_01_HDR_Volume}
  MaterialExpressionMultiply_2 = Multiply(
   A=Subtract(A=SceneDepthWithoutWater!(UVs).R, B=PixelDepth().R)
   B=ScalarParameter() {defaultValue=0.0125, parameterName=Caustics Mip Density}
  )
  n17 = Custom!(n=@MaterialExpressionMultiply_2) {
   code="return log2(n);"
   description=log2n
   outputType=CMOT_Float1
   inputs [{inputName=n, input {expression=&self:MaterialExpressionMultiply_2}}]
  }
  n18 = DeriveNormalZ(InXY=@n11.RGB)
  n19 = Multiply(
   A=Min(A=@n7, B=ScalarParameter() {defaultValue=8, parameterName=Sim Caustics Min Depth})
   B=ScalarParameter() {defaultValue=16384, parameterName=Sim Caustics Finite Offset Inv}
  )
  n20 = TextureSampleParameter2D(UVs=Add(A=@n10.UV, B=Divide(B=@n19))) {
   parameterName=NormalAndHeight
   samplerSource=SSM_Clamp_WorldGroupSettings
   samplerType=SAMPLERTYPE_LinearColor
   texture=&Black_1x1_EXR_Texture
  }
  n21 = DeriveNormalZ(InXY=@n20.RGB)
  n22 = Abs(None=Subtract(A=ComponentMask(None=@n18) {b=true}, B=ComponentMask(None=@n21) {b=true}))
  MaterialExpressionMultiply_7 = Multiply(A=@n22, B=ScalarParameter() {defaultValue=2048, parameterName=Sim Cautics Density})
  n23 = Custom!(x=@MaterialExpressionMultiply_7) {
   code="1-exp(-x);"
   description=exp
   outputType=CMOT_Float1
   inputs [{inputName=x, input {expression=&self:MaterialExpressionMultiply_7}}]
  }
  n24 = Add(
   A=TextureSample!(
    UVs=AppendVector(A=@n14, B=@n15.Frac Time)
    Tex=FunctionInput(Preview=@n16) {inputName=Volume Texture, inputType=FunctionInput_VolumeTexture}
    Level=Max(A=@n17) {constB=0}
    Apply View MipBias
   ) {mipValueMode=TMVM_MipLevel, texture=&Caustics_Animated_01_HDR_Volume}.RGB
   B=Multiply(A=@n23, B=ScalarParameter() {defaultValue=6, parameterName=Sim Caustics})
  )
  n25 = FunctionInput() {bUsePreviewValueAsDefault=true, inputName=Caustics Brightness, inputType=FunctionInput_Scalar, sortPriority=4, previewValue=<1 0 0 1>}
  n26 = DotProduct(
   A=MaterialFunctionCall!(ViewportUV=@n5.Result) {materialFunction=/Water/Materials/Functions/WorldNormalBehindWater}.WorldNormal
   B=SkyAtmosphereLightDirection()
  )
  FunctionOutput(None=LinearInterpolate(B=Multiply(A=@n24, B=@n25), Alpha=Max(A=@n26) {constB=0}) {constA=1})
  Saturate(None=@n22)
 ]
}
