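# Legacy material function that evaluates Gerstner waves from a wave-parameter render target, for custom uses like caustics, superseded by ComputeGerstnerWaves and unsuitable for SingleLayerWater.
# Custom HLSL using GerstnerWaveFunctions.ush computes packed waves from position, time and the texture.
# Two unpack nodes split this into Normal and WPO, also output as material attributes.
# Distance-field time-lag and wave-mask branches exist but reach no output.
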MaterialFunction /Water/Caustics/Materials/Functions/GetGerstnerWaves
graph {
 properties {description="This function uses the old method to get Wave data using a render target. As such, it does NOT work properly with SingleLayerWater materials, it is only useful for custom cases such as caustics generation. Please use 'ComputeGerstnerWaves' instead."}
 nodes [
  MaterialExpressionFunctionInput_2 = FunctionInput(Preview=Time()) {bUsePreviewValueAsDefault=true, inputName=Time, inputType=FunctionInput_Scalar, sortPriority=2}
  n1 = WorldPosition() {worldPositionShaderOffset=WPT_ExcludeAllShaderOffsets}
  n2 = Max(A=DistanceToNearestSurface!(World Position=@n1.XYZ)) {constB=0}
  MaterialExpressionMultiply_0 = Multiply(A=@n2, B=ScalarParameter() {defaultValue=0.02, parameterName=Lag Density, sliderMin=50000})
  n3 = Custom!(x=@MaterialExpressionMultiply_0) {
   code="exp(-x);"
   outputType=CMOT_Float1
   inputs [{inputName=x, input {expression=&self:MaterialExpressionMultiply_0}}]
  }
  Add(A=@MaterialExpressionFunctionInput_2, B=Multiply(A=@n3, B=ScalarParameter() {defaultValue=-0.5, parameterName=Lag}))
  MaterialExpressionFunctionInput_0 = FunctionInput(Preview=TextureObject() {samplerType=SAMPLERTYPE_LinearColor, texture=/Water/Textures/DefaultWaveSpectrumRT}) {bUsePreviewValueAsDefault=true, inputName=WaveParameterRT, inputType=FunctionInput_Texture2D}
  MaterialExpressionFunctionInput_1 = FunctionInput(Preview=@n1.XYZ) {bUsePreviewValueAsDefault=true, inputName=Override Position, inputType=FunctionInput_Vector2, sortPriority=1}
  MaterialExpressionCustom_0 = Custom!(
   WaveParameterRT=@MaterialExpressionFunctionInput_0
   WorldPosition=@MaterialExpressionFunctionInput_1
   Time=@MaterialExpressionFunctionInput_2
  ) {
   code="GerstnerWaveRenderer WaveRenderer = InitializeGerstnerWaveRenderer(float2(64,1), 32, WorldPosition, Time);\r\n\r\nWaveOutput Waves;\r\n\r\nWaves = GetAllGerstnerWaves(WaveRenderer, WaveParameterRT, WaveParameterRTSampler, 0);\r\nreturn PackNormalAndWPO(Waves);"
   description=Get Gerstner Waves
   includeFilePaths [/Plugin/Water/Private/GerstnerWaveFunctions.ush]
   inputs [
    {inputName=WaveParameterRT, input {expression=&self:MaterialExpressionFunctionInput_0}}
    {inputName=WorldPosition, input {expression=&self:MaterialExpressionFunctionInput_1}}
    {inputName=Time, input {expression=&self:MaterialExpressionFunctionInput_2}}
   ]
  }
  n4 = Custom!(inPackedWave=@MaterialExpressionCustom_0) {
   code="\t\tfloat3 outnormal = frac(inPackedWave);\r\n\t\toutnormal *= 2.02;\r\n\t\toutnormal -= 1.01;\r\n\t\treturn outnormal;"
   description=UnpackNormal
   inputs [{inputName=inPackedWave, input {expression=&self:MaterialExpressionCustom_0}}]
  }
  FunctionOutput(None=@n4) {outputName=Normal, sortPriority=1}
  n5 = Custom(inPackedWave=@MaterialExpressionCustom_0) {
   code="\t\tfloat3 outWPO;\r\n\t\toutWPO = floor(inPackedWave) / 100;\r\n\t\treturn outWPO;"
   description=UnpackWPO
   inputs [{inputName=inPackedWave, input {expression=&self:MaterialExpressionCustom_0}}]
  }
  FunctionOutput(None=@n5) {outputName=WPO, sortPriority=2}
  FunctionOutput(None=SetMaterialAttributes!(MaterialAttributes, Normal=@n4, WorldPositionOffset=@n5) {attributeSetTypes [#1, #2]}) {outputName=Wave Attributes}
  n6 = Divide(A=@n2, B=ScalarParameter() {defaultValue=256, parameterName=Wave Mask Distance})
  Multiply(A=@n5, B=Max(A=Saturate(None=@n6), B=ScalarParameter() {defaultValue=0.6, parameterName=Min Waves near DF}))
 ]
}
