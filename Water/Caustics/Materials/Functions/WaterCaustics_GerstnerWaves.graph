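# Water plugin material function that computes analytical caustics brightness on surfaces seen through water from the Gerstner wave simulation.
# Refraction-distorted screen UVs give the position behind the water, which is shifted by depth, sky light direction and wave normal.
# Finite differences of depth-scaled wave normals at three offset samples estimate refracted area via a cross product.
# Undistorted-to-refracted area ratio, clamped by a minimum focus size, drives the brightness-scaled output.
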MaterialFunction /Water/Caustics/Materials/Functions/WaterCaustics_GerstnerWaves
graph {
 properties {
  bExposeToLibrary=true
  description=This is an analytical Caustics function. It requires the Water system. It will use finite differences to sample the wave function to genreate analyitcal caustics.
  libraryCategoriesText [Water]
 }
 nodes [
  n1 = FunctionInput() {bUsePreviewValueAsDefault=true, inputName=Refraction, inputType=FunctionInput_Scalar, previewValue=<1.05 0 0 1>}
  n2 = MaterialFunctionCall!(Refraction=@n1) {materialFunction=&/GetDistortionUV}
  n3 = MaterialFunctionCall!(ScreenUV=@n2.Result) {materialFunction=/Water/Materials/Functions/WorldPositionBehindWater}
  n4 = Subtract(
   A=ComponentMask(None=WorldPosition() {worldPositionShaderOffset=WPT_ExcludeAllShaderOffsets}.XYZ) {b=true}
   B=ComponentMask(None=@n3.Result) {b=true}
  )
  n5 = Add(
   A=Multiply(A=@n4, B=ScalarParameter() {parameterName=Water Depth Caustics Spread})
   B=ScalarParameter() {defaultValue=16, parameterName=Caustics Finite Offset}
  )
  n6 = Multiply(A=@n5)
  n7 = Multiply(A=@n6, B=@n6)
  n8 = Multiply(A=Add(A=Multiply(A=@n4) {constB=2, desc=Water Depth}, B=ScalarParameter() {defaultValue=64, parameterName=Caustics Water Depth Bias})) {constB=-1}
  n9 = Multiply(A=@n4, B=ScalarParameter() {defaultValue=2, parameterName=Refract Caustic Position})
  n10 = SkyAtmosphereLightDirection()
  n11 = Divide(A=ComponentMask(None=@n10) {g=true, r=true}, B=LinearInterpolate(B=ComponentMask(None=@n10) {b=true}) {constA=1, constAlpha=0.4})
  n12 = ScalarParameter() {defaultValue=1, parameterName=Light Parallax}
  n13 = Add(A=Multiply(A=Multiply(A=@n11, B=@n12), B=@n4), B=ComponentMask(None=@n3.Result) {g=true, r=true})
  n14 = StaticSwitchParameter(
   True=ScalarParameter() {parameterName=Caustics Waves Water Body Index}
   False=Custom() {code="GetWaterWaveParamIndex(Parameters)", description=WaterBodyIndex From Tile Instance, outputType=CMOT_Float1, inputs []}
  ) {parameterName=Caustics Use Manual Water Body Index}
  n15 = MaterialFunctionCall!(
   Water Body Index=@n14
   Override Position=@n13
   Time=LinearInterpolate(
    A=Time()
    B=CollectionParameter() {collection=/Water/Materials/MPC/MPC_Water, parameterName=ManualTime}
    Alpha=CollectionParameter() {collection=&MPC_Water, parameterName=FreezeTime}
   )
   Use Depth Attenuation=StaticBool()
  ) {materialFunction=/Water/Materials/Functions/ComputeGerstnerWaves}
  n16 = Add(A=ComponentMask(None=Multiply(A=@n9, B=@n15.Normal)) {g=true, r=true}, B=@n13)
  n17 = StaticBool()
  n18 = MaterialFunctionCall() {materialFunction=/Water/Materials/Functions/GetWaterTime}
  n19 = Multiply(
   A=@n8
   B=MaterialFunctionCall(Water Body Index=@n14, Override Position=@n16, Time=@n18.WaterTime, Use Depth Attenuation=@n17) {materialFunction=&ComputeGerstnerWaves}.Normal
  )
  n20 = Constant2Vector() {r=1}
  n21 = MaterialFunctionCall(Water Body Index=@n14, Override Position=Add(A=@n16, B=Multiply(A=@n5, B=@n20)), Time=@n18.WaterTime, Use Depth Attenuation=@n17) {materialFunction=&ComputeGerstnerWaves}
  n22 = Constant3Vector() {constant=<rgba 1 0 0 0>}
  n23 = Subtract(A=@n19, B=Add(A=Multiply(A=@n8, B=@n21.Normal), B=Multiply(A=@n5, B=@n22)))
  n24 = Constant3Vector() {constant=<rgba 1 1 0 0>}
  n25 = Constant2Vector() {g=1}
  n26 = MaterialFunctionCall(Water Body Index=@n14, Override Position=Add(A=@n16, B=Multiply(A=@n5, B=@n25)), Time=@n18.WaterTime, Use Depth Attenuation=@n17) {materialFunction=&ComputeGerstnerWaves}
  n27 = Constant3Vector() {constant=<rgba 0 1 0 0>}
  n28 = Subtract(A=Add(A=Multiply(A=@n8, B=@n26.Normal), B=Multiply(A=@n5, B=@n27)), B=@n19)
  n29 = MaterialFunctionCall(Vector 3=CrossProduct(A=Multiply(A=@n23, B=@n24), B=Multiply(A=@n28, B=@n24))) {materialFunction=/Engine/Functions/Engine_MaterialFunctions02/Utility/VectorLength}
  n30 = ScalarParameter() {defaultValue=0.0005, parameterName=Min focus size}
  n31 = LinearInterpolate(B=Divide(A=@n7, B=Max(A=@n29.V3 Length, B=@n30))) {constA=1, desc=Fade}
  n32 = FunctionInput() {bUsePreviewValueAsDefault=true, inputName=Caustics Brightness, inputType=FunctionInput_Scalar, sortPriority=4, previewValue=<1 0 0 1>}
  FunctionOutput(None=LinearInterpolate(B=Multiply(A=@n31, B=@n32)) {constA=1, constAlpha=1}) {outputName=Caustics, sortPriority=1}
  Multiply(A=ScalarParameter() {defaultValue=1, parameterName=Caustics Brightness Adjust}, B=@n7)
  ObjectPositionWS()
  n33 = DotProduct(
   A=MaterialFunctionCall!(ViewportUV=@n2.Result) {materialFunction=/Water/Materials/Functions/WorldNormalBehindWater}.WorldNormal
   B=SkyAtmosphereLightDirection()
  )
  Reroute(None=StaticSwitch(True=Max(A=@n33) {constB=0}, False=Constant() {r=1}, Value=StaticBool()))
 ]
}
